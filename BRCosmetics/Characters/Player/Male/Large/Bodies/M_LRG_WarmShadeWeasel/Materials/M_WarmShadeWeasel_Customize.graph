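# Customizable character material for the WarmShadeWeasel large male body, with recolorable regions and animated-looking energy flow effects.
# Packed masks tint a desaturated base color by primary, secondary, accent and energy colors, with optional subsurface recoloring and thin-film LUT modulation.
# Gradient-driven flow functions, optionally doubled with an offset, vary length by pre-skinned X position.
# Material attributes are overridden between shared CharacterParent PreFX and PostFX functions, including a masked shading-model blend.
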Material /BRCosmetics/Characters/Player/Male/Large/Bodies/M_LRG_WarmShadeWeasel/Materials/M_WarmShadeWeasel_Customize
graph {
 properties {
  bUseMaterialAttributes=true
  bUsedWithMeshDeformer=true
  bUsedWithMorphTargets=true
  bUsedWithSkeletalMesh=true
  pixelDepthOffsetMode=PDOM_Legacy
  shadingModel=MSM_FromMaterialExpression
  usedShadingModels=Default Lit | Subsurface | Clear Coat | Subsurface Profile | Cloth
 }
 nodes [
  n1 = Multiply(A=MaterialFunctionCall() {materialFunction=&Fresnel_Function}.Result) {constB=3}
  n2 = Time()
  Add(A=@n1, B=Multiply(A=ScalarParameter() {defaultValue=0.5, group=CustomVariation, parameterName=NoiseMovingSpeed}, B=@n2))
  n3 = NamedRerouteUsage()
  n4 = NamedRerouteUsage()
  n5 = NamedRerouteUsage()
  n6 = NamedRerouteUsage()
  n7 = ScalarParameter() {defaultValue=0.2, parameterName=AccentSSSMulti}
  n8 = NamedRerouteUsage()
  n9 = LinearInterpolate(
   A=LinearInterpolate(A=LinearInterpolate(A=@n3, B=@n4, Alpha=@n5), B=NamedRerouteUsage(), Alpha=NamedRerouteUsage())
   B=Saturate(None=Multiply(A=@n6, B=@n7))
   Alpha=Add(A=NamedRerouteUsage(), B=@n8)
  )
  LinearInterpolate(A=@n9, B=NamedRerouteUsage(), Alpha=@n8)
  n10 = MaterialFunctionCall() {materialFunction=&MF_CharacterParent_2023_PreFX}
  n11 = GetMaterialAttributes(None=@n10.Result) {attributeGetTypes [
    #1
    #2
    #3
    #4
    #5
   ]}
  n12 = NamedRerouteUsage()
  n13 = Multiply(
   A=Multiply(A=@n11.Emissive Color, B=NamedRerouteUsage())
   B=ScalarParameter() {defaultValue=1, group=CustomVariation, parameterName=EmissiveIntensityRE}
  )
  n14 = NamedRerouteUsage()
  n15 = SetMaterialAttributes!(
   MaterialAttributes=@n11.MaterialAttributes
   Base Color=NamedRerouteUsage()
   Emissive Color=LinearInterpolate(A=Constant(), B=Add(A=@n13, B=@n14), Alpha=NamedRerouteUsage())
   Fuzz Color or Subsurface Color=NamedRerouteUsage()
   Roughness=LinearInterpolate(A=@n11.Roughness, Alpha=@n12) {constB=0.4}
   Metallic=LinearInterpolate(A=@n11.Metallic, Alpha=@n12)
  ) {attributeSetTypes [
    #1
    #2
    #3
    #4
    #5
   ]}
  n16 = ShadingModel()
  MaterialFunctionCall(
   Material=BlendMaterialAttributes(
    A=@n15
    B=SetMaterialAttributes!(MaterialAttributes=@n15, Shading Model=@n16) {attributeSetTypes [#6]}
    Alpha=NamedRerouteUsage()
   )
   M=@n10.M
   Pre Skinned Local Position=@n10.M
   Distance Blend=@n10.M
   Fresnel=@n10.M
  ) {materialFunction=&MF_CharacterParent_2023_PostFX}
  n17 = NamedRerouteUsage()
  n18 = Multiply(A=@n17, B=NamedRerouteUsage())
  n19 = TextureSampleParameter2D(UVs=Multiply(A=TextureCoordinate(), B=ScalarParameter() {defaultValue=8, group=CustomVariation, parameterName=GridTiling})) {
   group=CustomVariation
   parameterName=GridNoise
   texture=/BRCosmetics/Characters/Player/Male/Large/Bodies/M_LRG_WarmShadeWeasel/Textures/T_GridNoise
  }
  n20 = Add(A=Multiply(A=@n2, B=ScalarParameter() {defaultValue=0.2, group=CustomVariation, parameterName=NoiseChangeSpeed}), B=@n19.R)
  n21 = SmoothStep(Value=@n19.R) {constMin=0.7}
  Multiply(A=@n18, B=Multiply(A=Frac(None=@n20), B=OneMinus(None=@n21)))
  Multiply(A=ConstantBiasScale(None=Sine(None=@n1)), B=@n17)
  n22 = TextureSampleParameter2D() {
   group=CustomVariation
   parameterName=01_Prim_Second_Mask
   samplerType=SAMPLERTYPE_Masks
   texture=/BRCosmetics/Characters/Player/Male/Large/Bodies/M_LRG_WarmShadeWeasel/Textures/T_WarmShadeWeasel_Body_FX2
  }
  NamedRerouteDeclaration(None=@n22.R) {name=Prim_Area, nodeColor=<rgba 1 0.0234 0 1>}
  NamedRerouteDeclaration(None=@n22.G) {name=Second_Area, nodeColor=<rgba 0 0.2969 1 1>}
  n23 = NamedRerouteUsage()
  NamedRerouteDeclaration(None=Multiply(A=@n23, B=VectorParameter() {group=CustomVariation, parameterName=Secondary_Color, defaultValue=<rgba 1 0 0 1>}.RGB)) {name=Second_Tint, nodeColor=<rgba 0.9688 1 0 1>}
  n24 = NamedRerouteUsage()
  n25 = VectorParameter() {group=CustomVariation, parameterName=Energy_Color, defaultValue=<rgba 0 1 0 1>}
  NamedRerouteDeclaration(None=Multiply(A=@n24, B=NamedRerouteDeclaration(None=@n25.RGB) {name=EnergyColor, nodeColor=<rgba 1 0 0.5625 1>})) {name=Energy_Tint, nodeColor=<rgba 0.5469 0 1 1>}
  n26 = NamedRerouteUsage()
  NamedRerouteDeclaration(None=StaticSwitchParameter(True=StaticSwitchParameter(True=@n9, False=@n26) {group=CustomVariation, parameterName=UseCustomSSSColor}, False=NamedRerouteUsage()) {defaultValue=true, group=CustomVariation, parameterName=UseCustomColor}) {name=SSS_OUT, nodeColor=<rgba 0 1 0.2266 1>}
  n27 = NamedRerouteUsage()
  n28 = CameraVectorWS()
  n29 = PixelNormalWS()
  n30 = TextureSampleParameter2D(UVs=Clamp(None=DotProduct(A=@n28, B=@n29))) {
   group=06_Thin Film
   parameterName=Thin Film Texture
   sortPriority=1
   texture=/BRCosmetics/Characters/Player/Male/Large/Bodies/M_LRG_WarmShadeWeasel/Textures/T_GenericThinFilm
  }
  n31 = ChannelMaskParameter(None=@n30.RGB) {group=06_Metal LUT, parameterName=LUTChannel}
  n32 = Saturate(None=ScalarParameter() {defaultValue=1, group=06_Metal LUT, parameterName=MetalLutIntensity})
  NamedRerouteDeclaration(None=LinearInterpolate(A=@n27, B=Multiply(A=@n31.R, B=@n27), Alpha=@n32)) {name=EnergyColor_LUT, nodeColor=<rgba 0.8281 1 0 1>}
  n33 = TextureSampleParameter2D() {
   group=CustomVariation
   parameterName=02_Accent_Energy_Mask
   samplerType=SAMPLERTYPE_Masks
   texture=/BRCosmetics/Characters/Player/Male/Large/Bodies/M_LRG_WarmShadeWeasel/Textures/T_WarmShadeWeasel_Body_FX3
  }
  NamedRerouteDeclaration(None=@n33.B) {name=TapeArea, nodeColor=<rgba 0.3359 1 0 1>}
  n34 = ScalarParameter() {defaultValue=2, parameterName=EnergySubtleIntensity}
  NamedRerouteDeclaration(None=Multiply(A=Multiply(A=@n18, B=@n34) {constB=0.2}, B=NamedRerouteUsage())) {name=Energy_Subtle, nodeColor=<rgba 1 0 0.6562 1>}
  n35 = NamedRerouteUsage()
  n36 = MaterialFunctionCall(Float3=@n35) {materialFunction=&BreakOutFloat3Components}
  n37 = Max(A=@n36.R, B=@n36.R)
  n38 = Multiply(A=Normalize(VectorInput=@n35)) {constB=0.1}
  NamedRerouteDeclaration(None=If(A=Max(A=@n37, B=@n36.B), B=Constant() {r=0.05}, "A > B"=@n35, "A == B"=@n38, "A < B"=@n38)) {name=MinEnergy, nodeColor=<rgba 0.9922 1 0 1>}
  NamedRerouteDeclaration(None=@n33.R) {name=Accent_Area, nodeColor=<rgba 0.6172 1 0 1>}
  n39 = TextureSampleParameter2D() {
   group=CustomVariation
   parameterName=GradientMap
   samplerType=SAMPLERTYPE_LinearGrayscale
   texture=/BRCosmetics/Characters/Player/Male/Large/Bodies/M_LRG_WarmShadeWeasel/Textures/T_WarmShadeWeasel_Body_FX_Gradient
  }
  n40 = ScalarParameter() {defaultValue=0.3, group=CustomVariation, parameterName=SimpleGradientMin}
  n41 = ScalarParameter() {defaultValue=0.7, group=CustomVariation, parameterName=SimpleGradientMax}
  n42 = StaticComponentMaskParameter(None=TextureCoordinate()) {defaultG=true, group=CustomVariation, parameterName=GradientDirection}
  n43 = NamedRerouteDeclaration(None=StaticSwitchParameter(True=Saturate(None=@n39.R), False=SmoothStep(Min=@n40, Max=@n41, Value=@n42)) {defaultValue=true, group=CustomVariation, parameterName=UseComplexGradient}) {name=Gradient, nodeColor=<rgba 0.8516 0 1 1>}
  n44 = TextureSampleParameter2D() {
   group=CustomVariation
   parameterName=GradientMap_Area
   texture=/BRCosmetics/Characters/Player/Male/Large/Bodies/M_LRG_WarmShadeWeasel/Textures/T_GradientSP_Body_Area
  }
  n45 = StaticSwitchParameter(True=@n44.R, False=Constant() {r=1}) {defaultValue=true, group=CustomVariation, parameterName=UseGradientAreaMask}
  n46 = NamedRerouteUsage()
  n47 = Constant()
  n48 = If(
   A=Abs(None=ComponentMask(None=@n46) {r=true})
   B=ScalarParameter() {defaultValue=50, parameterName=BodyInsideCM}
   "A > B"=Constant() {r=1}
   "A == B"=@n47
   "A < B"=@n47
  )
  n49 = Constant()
  n50 = LinearInterpolate(
   A=ScalarParameter() {defaultValue=0.2, group=Flow_Override, parameterName=FlowLengthMin}
   B=ScalarParameter() {defaultValue=0.4, group=Flow_Override, parameterName=FlowLengthMax}
   Alpha=NamedRerouteDeclaration(None=StaticSwitchParameter(True=@n48, False=@n49) {defaultValue=true, group=CustomVariation, parameterName=UseArmAlpha}) {name=PreskinnedArmAlpha, nodeColor=<rgba 1 0.7734 0 1>}
  )
  n51 = MaterialFunctionCall!(FlowGradient=@n43, FlowMask=@n45, FlowLength=@n50) {materialFunction=&/MF_Flow_WarmShade}
  n52 = Add(A=@n43) {constB=0.5}
  n53 = Add(
   A=@n51.Result
   B=MaterialFunctionCall(FlowGradient=Frac(None=@n52), FlowMask=@n45, FlowLength=@n50) {materialFunction=&MF_Flow_WarmShade}.Result
  )
  NamedRerouteDeclaration(None=Multiply(
    A=ScalarParameter() {defaultValue=1, group=CustomVariation, parameterName=FlowIntensityOverall}
    B=StaticSwitchParameter(True=@n53, False=@n51.Result) {defaultValue=true, group=Flow_Override, parameterName=UseDoubleFlow}
   )) {name=FlowColor, nodeColor=<rgba 0.6875 1 0 1>}
  NamedRerouteDeclaration(None=@n10.Pre skinned Local Position) {name=PreskinnedPOS, nodeColor=<rgba 0 0.9062 1 1>}
  n54 = NamedRerouteUsage()
  n55 = MaterialFunctionCall(Float3=@n54) {materialFunction=&BreakOutFloat3Components}
  n56 = Max(A=Max(A=@n55.R, B=@n55.R), B=@n55.B)
  n57 = Constant() {r=0.05}
  n58 = Multiply(A=Normalize(VectorInput=@n54)) {constB=0.1}
  NamedRerouteDeclaration(None=If(A=@n56, B=Constant() {r=0.01}, "A > B"=If(A=@n56, B=@n57, "A > B"=@n54, "A == B"=@n58, "A < B"=@n58), "A == B"=@n54, "A < B"=@n54)) {name=MinEnergy_Basecolor, nodeColor=<rgba 1 0 0.0234 1>}
  NamedRerouteDeclaration(None=@n33.G) {name=Energy_Area, nodeColor=<rgba 1 0 0.9375 1>}
  NamedRerouteDeclaration(None=@n11.Base Color) {name=BaseColorIN, nodeColor=<rgba 0 1 0.7422 1>}
  n59 = NamedRerouteUsage()
  n60 = Multiply(A=VectorParameter() {group=CustomVariation, parameterName=Primary_Color, defaultValue=<rgba 0.8416 0 1 1>}.RGB, B=@n23)
  n61 = ScalarParameter() {defaultValue=1, parameterName=PrimColorIntensity}
  n62 = NamedRerouteDeclaration(None=Saturate(None=Multiply(A=@n60, B=@n61))) {name=Prim_Tint, nodeColor=<rgba 0 0.6484 1 1>}
  n63 = NamedRerouteUsage()
  n64 = Multiply(A=VectorParameter() {group=CustomVariation, parameterName=Accent_Color, defaultValue=<rgba 0 0.9118 1 1>}.RGB, B=@n24)
  n65 = ScalarParameter() {defaultValue=0.7, parameterName=AccentColorIntensity}
  n66 = LinearInterpolate(
   A=LinearInterpolate(A=LinearInterpolate(A=@n59, B=@n62, Alpha=@n63), B=NamedRerouteUsage(), Alpha=NamedRerouteUsage())
   B=NamedRerouteDeclaration(None=Multiply(A=@n64, B=@n65)) {name=Accent_Tint, nodeColor=<rgba 1 0.8672 0 1>}
   Alpha=NamedRerouteUsage()
  )
  n67 = NamedRerouteUsage()
  n68 = NamedRerouteUsage()
  NamedRerouteDeclaration(None=StaticSwitchParameter(True=LinearInterpolate(A=@n66, B=@n67, Alpha=@n68), False=NamedRerouteUsage()) {defaultValue=true, group=CustomVariation, parameterName=UseCustomColor}) {name=BaseColor_OUT, nodeColor=<rgba 1 0.4219 0 1>}
  n69 = TextureSampleParameter2D() {
   group=CustomVariation
   parameterName=Desaturated_BaseColor
   texture=/BRCosmetics/Characters/Player/Male/Large/Bodies/M_LRG_WarmShadeWeasel/Textures/T_WarmShadeWeasel_Body_DS
  }
  NamedRerouteDeclaration(None=StaticSwitchParameter(True=Constant() {r=1}, False=@n69.RGB) {group=CustomVariation, parameterName=DebugPureColor}) {name=Desaturated_Color, nodeColor=<rgba 0.9688 0 1 1>}
  NamedRerouteDeclaration(None=@n11.Fuzz Color or Subsurface Color) {name=SSSColorIN, nodeColor=<rgba 0.0078 1 0 1>}
  n70 = NamedRerouteUsage()
  NamedRerouteDeclaration(None=LinearInterpolate(A=@n70, B=Multiply(A=@n70, B=@n31.R), Alpha=@n32)) {name=SecondColor_LUT, nodeColor=<rgba 0 1 0.2969 1>}
  NamedRerouteUsage()
  SmoothStep(Value=@n44.R) {constMax=0.5, constMin=0.1}
 ]
}
