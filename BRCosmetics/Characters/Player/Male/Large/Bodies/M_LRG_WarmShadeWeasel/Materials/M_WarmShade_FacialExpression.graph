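# Character material for the WarmShadeWeasel body that can blend an extra facial expression normal map over the standard character look.
# It runs the 2023 character parent PreFX function, then lerps its normal toward FacialExpression_N by the saturated Normal_Curve scalar.
# The UseFacialExpression static switch picks the blended or untouched attributes and passes them to the parent PostFX function.
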Material /BRCosmetics/Characters/Player/Male/Large/Bodies/M_LRG_WarmShadeWeasel/Materials/M_WarmShade_FacialExpression
graph {
 properties {
  bUseMaterialAttributes=true
  bUsedWithMeshDeformer=true
  bUsedWithSkeletalMesh=true
  pixelDepthOffsetMode=PDOM_Legacy
  shadingModel=MSM_FromMaterialExpression
  usedShadingModels=Default Lit | Subsurface | Clear Coat | Subsurface Profile | Cloth
 }
 nodes [
  n1 = MaterialFunctionCall() {materialFunction=&MF_CharacterParent_2023_PreFX}
  n2 = GetMaterialAttributes(None=@n1.Result) {attributeGetTypes [#1]}
  n3 = TextureSampleParameter2D() {
   group=001_FacialExpression
   parameterName=FacialExpression_N
   samplerType=SAMPLERTYPE_Normal
   texture=&FlatNormalBlue
  }
  n4 = LinearInterpolate(A=@n2.Normal, B=@n3.RGB, Alpha=Saturate(None=ScalarParameter() {group=001_FacialExpression, parameterName=Normal_Curve}))
  MaterialFunctionCall(
   Material=StaticSwitchParameter(
    True=SetMaterialAttributes!(MaterialAttributes=@n2.MaterialAttributes, Normal=@n4) {attributeSetTypes [#1]}
    False=@n1.Result
   ) {group=001_FacialExpression, parameterName=UseFacialExpression}
   M=@n1.M
   Pre Skinned Local Position=@n1.M
   Distance Blend=@n1.M
   Fresnel=@n1.M
  ) {materialFunction=&MF_CharacterParent_2023_PostFX}
 ]
}
