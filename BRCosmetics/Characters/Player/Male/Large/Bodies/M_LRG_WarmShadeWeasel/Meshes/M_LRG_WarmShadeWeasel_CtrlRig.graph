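# Control Rig for the WarmShadeWeasel large male player body, adding procedural correctives to deform bones and clothing bones.
# A shared DeformRig function runs first, then sequenced branches handle feet and shoe tongues, shorts, jacket, sleeves, neck and buckle.
# Remapped joint angles, thigh-direction angle tests and a spherical pose reader weight the offset transforms.
# Parent constraints steer jacket and shorts bones, and a "falling alpha" variable blends a jacket override pose.
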ControlRigBlueprint /BRCosmetics/Characters/Player/Male/Large/Bodies/M_LRG_WarmShadeWeasel/Meshes/M_LRG_WarmShadeWeasel_CtrlRig
graphs [
 Maximum_1_ContainedGraph {nodes [
   n1 = RigVMFunctionEntryNode()
   Maximum = &RigVMFunction_MathDoubleMax(A=@n1.A, B=@n1.B)
   RigVMFunctionReturnNode(Result=&RigVMFunction_MathDoubleMax(B=@n1.C, A=@Maximum.Result).Result)
  ]}
 RigVMFunctionLibrary {links [], nodes []}
 RigVMFunction_Sequence_1_1_ContainedGraph {nodes [
   n1 = RigVMFunctionEntryNode()
   n2 = &RigVMFunction_Sequence(ExecuteContext=@n1.ExecuteContext)
   n3 = &RigVMFunction_Sequence(ExecuteContext=@n2.B)
   RigVMFunctionReturnNode(A=@n2.A, B=@n3.A, C=@n3.B)
  ]}
 RigVMFunction_Sequence_1_ContainedGraph {nodes [
   n1 = RigVMFunctionEntryNode()
   RigVMFunction_Sequence = &RigVMFunction_Sequence(ExecuteContext=@n1.ExecuteContext)
   n2 = &RigVMFunction_Sequence(ExecuteContext=@RigVMFunction_Sequence.B)
   n3 = &RigVMFunction_Sequence(ExecuteContext=@n2.B)
   n4 = &RigVMFunction_Sequence(ExecuteContext=@n3.B)
   n5 = &RigVMFunction_Sequence(ExecuteContext=@n4.B)
   n6 = &RigVMFunction_Sequence(ExecuteContext=@n5.B)
   RigVMFunctionReturnNode(A=@RigVMFunction_Sequence.A, B=@n2.A, C=@n3.A, D=@n4.A, E=@n5.A, F=@n6.A, H=@n6.B)
  ]}
 RigVMModel {nodes [
   RigUnit_BeginExecution = &RigUnit_BeginExecution()
   FN_DeformRig_FNC = RigVMFunctionReferenceNode(
    DeformRig_ Curve="(deformRig_l_upperarm_up_alpha_72_#1=False,deformRig_r_upperarm_up_alpha_73_#2=False,deformRig_l_upperarm_fr_alpha_74_#3=False,deformRig_r_upperarm_fr_alpha_75_CB51B8 ...(truncated)"
    DeformRig_Bool="(upperarm_1_#4=true,elbow_3_#5=True,wrist_7_#6=True,groin_8_#7=True,glute_17_#8=True,knee_9_40C408324CA261 ...(truncated)"
    DeformRig_Debug=(upperarm_debug_18_#4=false,wrist_debug_20_#6=False,lat_debug_26_#9=False,groinGlute_debug_27_#7=False)
    DeformRig_Value="(upperarm_up_18_#10=(Rotation=(X=0,Y=0,Z=0,W=1),Translation=(X=0,Y=0,Z=1),Scale3D=(X=0.7,Y=1,Z=1)),upperarm_fwd_fr_19_#11=(Rotation=(X=0,Y=0,Z=0,W=1),Translation=(X=0,Y=1,Z=0),Scale3D=(X=0.5,Y=1,Z=1 ...(truncated)"
    ExecuteContext=@RigUnit_BeginExecution.ExecutePin
   )
   > RigVMFunction_Sequence = RigVMAggregateNode()
   n1 = &RigUnit_GetRelativeTransformForItem(Child=(Type=Bone,Name="hand_r"), Parent=(Type=Bone,Name="lowerarm_r"), bChildInitial=False, bParentInitial=False)
   &RigUnit_ModifyTransforms(
    ItemToModify=((Item=(Type=Bone,Name="hand_sleeve_r"),Transform=(Rotation=(X=0,Y=0,Z=0,W=-1),Translation=(X=-0.6197,Y=0,Z=-1.6594),Scale3D=(X=0.8991,Y=1,Z=1.4842))))
    WeightMinimum=0
    Weight=&RigVMFunction_MathDoubleRemap(
     SourceMaximum=-70
     SourceMinimum=-30
     TargetMaximum=1
     TargetMinimum=0
     Value=&RigVMFunction_MathQuaternionToEuler(RotationOrder=ZYX, Value=@n1.RelativeTransform.Rotation).Result.Y
    ).Result
    ExecutePin=@RigVMFunction_Sequence.D
   )
   n2 = &RigUnit_GetRelativeTransformForItem(Child=(Type=Bone,Name="hand_l"), Parent=(Type=Null,Name="hand_l_null"), bChildInitial=False, bParentInitial=False)
   > &RigUnit_ModifyTransforms(
    ItemToModify="((Item=(Type=Bone,Name=\"hand_sleeve_l\"),Transform=(Rotation=(X=-0.2588,Y=0,Z=0,W=0.9659),Translation=(X=0,Y=0,Z=-1),Scale3D=(X=1.2,Y=1.2,Z=1.2))),(Item=(Type=Bone,Name=\"lowerarm_twist_01_l\"),Transform=(Rotation=(X=-0.5,Y=0,Z=0,W=0.866),Translation=(X=0,Y=0 ...(truncated)"
    WeightMinimum=0
    Weight=&RigVMFunction_MathDoubleRemap(
     SourceMaximum=150
     SourceMinimum=50
     TargetMaximum=1
     TargetMinimum=0
     Value=&RigVMFunction_MathQuaternionToEuler(RotationOrder=XYZ, Value=@n2.RelativeTransform.Rotation).Result.X
    ).Result
   )
   n3 = &RigVMFunction_MathQuaternionToEuler(
    RotationOrder=ZYX
    Value=&RigUnit_GetRelativeTransformForItem(Child=(Type=Bone,Name="head"), Parent=(Type=Bone,Name="spine_05"), bChildInitial=False, bParentInitial=False).RelativeTransform.Rotation
   )
   Maximum = RigVMAggregateNode(
    A=&RigVMFunction_MathDoubleRemap(SourceMaximum=-30, SourceMinimum=4, TargetMaximum=1, TargetMinimum=0, Value=@n3.Result.Z).Result
    B=&RigVMFunction_MathDoubleRemap(SourceMaximum=-20, SourceMinimum=0, TargetMaximum=1, TargetMinimum=0, Value=@n3.Result.X).Result
    C=&RigVMFunction_MathDoubleRemap(SourceMaximum=-20, SourceMinimum=20, TargetMaximum=1, TargetMinimum=0, Value=@n3.Result.X).Result
   )
   &RigUnit_ModifyTransforms(
    ItemToModify=((Item=(Type=Bone,Name="mstr_neck_01"),Transform=(Rotation=(X=0,Y=0,Z=0,W=1),Translation=(X=-1,Y=1.0985,Z=0),Scale3D=(X=1,Y=1.2,Z=1.0476))))
    WeightMinimum=0
    Weight=@Maximum.Result
    ExecutePin=@RigVMFunction_Sequence.E
   )
   n4 = &RigUnit_SphericalPoseReader(
    ActiveRegionScaleFactors=(PositiveWidth=1,NegativeWidth=1,PositiveHeight=1,NegativeHeight=1)
    ActiveRegionSize=0.05
    Debug=(bDrawDebug=false,bDraw2D=False,bDrawLocalAxes=False,DebugScale=25,DebugSegments=20,DebugThickness=0.25)
    DriverAxis=(X=1,Y=0,Z=0)
    DriverItem=(Type=Bone,Name="hand_l")
    FalloffRegionScaleFactors=(PositiveWidth=1,NegativeWidth=1,PositiveHeight=1,NegativeHeight=1)
    FalloffSize=0.2
    FlipHeightScaling=False
    FlipWidthScaling=False
    OptionalParentItem=(Type=Bone,Name="None")
    RotationOffset=(X=0,Y=-90,Z=0)
    ExecutePin=@RigVMFunction_Sequence.F
   )
   Subtract = &RigVMFunction_MathDoubleSub(A=1, B=@n4.OutputParam)
   > &RigUnit_ModifyTransforms(
    ItemToModify=((Item=(Type=Bone,Name="hand_sleeve_l"),Transform=(Rotation=(X=0,Y=0,Z=0,W=1),Translation=(X=0,Y=0,Z=0),Scale3D=(X=1,Y=1,Z=1.3))))
    WeightMinimum=0
    Weight=@Subtract.Result
   )
   RigUnit_GetRelativeTransformForItem = &RigUnit_GetRelativeTransformForItem(Child=(Type=Bone,Name="thigh_l"), Parent=(Type=Null,Name="thigh_l_null"), bChildInitial=False, bParentInitial=False)
   RigVMFunction_MathQuaternionToEuler = &RigVMFunction_MathQuaternionToEuler(RotationOrder=ZYX, Value=@RigUnit_GetRelativeTransformForItem.RelativeTransform.Rotation)
   Remap = &RigVMFunction_MathDoubleRemap(SourceMaximum=-50, SourceMinimum=-15, TargetMaximum=1, TargetMinimum=0, bClamp=false, Value=@RigVMFunction_MathQuaternionToEuler.Result.Z)
   n5 = &RigVMFunction_MathQuaternionToEuler(
    RotationOrder=ZYX
    Value=&RigUnit_GetRelativeTransformForItem(Child=(Type=Bone,Name="thigh_r"), Parent=(Type=Null,Name="thigh_r_null"), bChildInitial=False, bParentInitial=False).RelativeTransform.Rotation
   )
   &RigUnit_ModifyTransforms(
    ItemToModify=((Item=(Type=Bone,Name="mstr_buckle"),Transform=(Rotation=(X=0,Y=0,Z=0,W=1),Translation=(X=1.2,Y=0.3,Z=0),Scale3D=(X=1,Y=1,Z=1))))
    WeightMinimum=0
    Weight=&RigVMFunction_MathDoubleMax(
     A=@Remap.Result
     B=&RigVMFunction_MathDoubleRemap(SourceMaximum=-50, SourceMinimum=-15, TargetMaximum=1, TargetMinimum=0, bClamp=False, Value=@n5.Result.Z).Result
    ).Result
    ExecutePin=@RigVMFunction_Sequence.H
   )
   n6 = &RigVMFunction_MathQuaternionToEuler(
    RotationOrder=ZYX
    Value=&RigUnit_GetRelativeTransformForItem(Child=(Type=Bone,Name="foot_r"), Parent=(Type=Null,Name="foot_r_null"), bChildInitial=False, bParentInitial=False).RelativeTransform.Rotation
   )
   &RigUnit_ModifyTransforms(
    ItemToModify=((Item=(Type=Bone,Name="mstr_foot_r"),Transform=(Rotation=(X=0,Y=0,Z=0.2823,W=-0.9593),Translation=(X=8.4793,Y=9.5152,Z=0),Scale3D=(X=1,Y=1,Z=1))))
    WeightMinimum=0
    Weight=&RigVMFunction_MathDoubleRemap(SourceMaximum=75, SourceMinimum=0, TargetMaximum=1, TargetMinimum=0, bClamp=False, Value=@n6.Result.Z).Result
    ExecutePin=@RigVMFunction_Sequence.A
   )
   > &RigUnit_ModifyTransforms(
    ItemToModify=((Item=(Type=Bone,Name="mstr_shoe_tongue_r"),Transform=(Rotation=(X=0.0094,Y=0.0028,Z=0.2864,W=0.9581),Translation=(X=-4.6581,Y=3.5102,Z=-0.4929),Scale3D=(X=1,Y=1.2,Z=1))))
    WeightMinimum=0
    Weight=&RigVMFunction_MathDoubleRemap(SourceMaximum=-50, SourceMinimum=0, TargetMaximum=1, TargetMinimum=0, bClamp=False, Value=@n6.Result.Z).Result
   )
   n7 = &RigVMFunction_MathQuaternionToEuler(
    RotationOrder=ZYX
    Value=&RigUnit_GetRelativeTransformForItem(Child=(Type=Bone,Name="foot_l"), Parent=(Type=Null,Name="foot_l_null"), bChildInitial=False, bParentInitial=False).RelativeTransform.Rotation
   )
   > &RigUnit_ModifyTransforms(
    ItemToModify=((Item=(Type=Bone,Name="mstr_foot_l"),Transform=(Rotation=(X=0,Y=0,Z=0.2823,W=-0.9593),Translation=(X=-8.4793,Y=-9.5152,Z=0),Scale3D=(X=1,Y=1,Z=1))))
    WeightMinimum=0
    Weight=&RigVMFunction_MathDoubleRemap(SourceMaximum=75, SourceMinimum=0, TargetMaximum=1, TargetMinimum=0, bClamp=False, Value=@n7.Result.Z).Result
   )
   > &RigUnit_ModifyTransforms(
    ItemToModify=((Item=(Type=Bone,Name="mstr_shoe_tongue_l"),Transform=(Rotation=(X=0.0094,Y=0.0028,Z=0.2864,W=0.9581),Translation=(X=4.6581,Y=-3.5102,Z=-0.4929),Scale3D=(X=1,Y=1.2,Z=1))))
    WeightMinimum=0
    Weight=&RigVMFunction_MathDoubleRemap(SourceMaximum=-50, SourceMinimum=0, TargetMaximum=1, TargetMinimum=0, bClamp=False, Value=@n7.Result.Z).Result
   )
   DISPATCH_RigVMDispatch_MakeStruct = RigVMDispatchNode(Elements=(TranslationFilter=(bX=false,bY=True,bZ=True),RotationFilter=(bX=true,bY=true,bZ=true),ScaleFilter=(bX=True,bY=True,bZ=True))) {template="DISPATCH_RigVMDispatch_MakeStruct(in Elements,out Struct)"}
   n8 = RigVMDispatchNode(Elements=(Item=(Type=Bone,Name="pelvis"),Weight=1)) {template="DISPATCH_RigVMDispatch_MakeStruct(in Elements,out Struct)"}
   RigVMFunction_MathFloatMake = /Script/RigVM.RigVMFunction_MathFloatMake(Value=1) {is=struct RigVMUnitNode}
   &RigUnit_ParentConstraint(
    AdvancedSettings=(InterpolationType=Average,RotationOrderForFilter=XZY)
    Child=(Type=Bone,Name="mstr_jacket_side_l")
    Parents=((Item=(Type=Bone,Name="pelvis"),Weight=1))
    Filter=@DISPATCH_RigVMDispatch_MakeStruct.Struct
    Parents.0=@n8.Struct
    Weight=@RigVMFunction_MathFloatMake.Value
    ExecutePin=@RigVMFunction_Sequence.C
   )
   > &RigUnit_ParentConstraint(
    AdvancedSettings=(InterpolationType=Average,RotationOrderForFilter=XZY)
    Child=(Type=Bone,Name="mstr_jacket_back_l")
    Parents=((Item=(Type=Bone,Name="pelvis"),Weight=1))
    Filter=@DISPATCH_RigVMDispatch_MakeStruct.Struct
    Parents.0=@n8.Struct
    Weight=@RigVMFunction_MathFloatMake.Value
   )
   > &RigUnit_ParentConstraint(
    AdvancedSettings=(InterpolationType=Average,RotationOrderForFilter=XZY)
    Child=(Type=Bone,Name="mstr_jacket_back_c")
    Parents=((Item=(Type=Bone,Name="pelvis"),Weight=1))
    Filter=@DISPATCH_RigVMDispatch_MakeStruct.Struct
    Parents.0=@n8.Struct
    Weight=@RigVMFunction_MathFloatMake.Value
   )
   > &RigUnit_ParentConstraint(
    AdvancedSettings=(InterpolationType=Average,RotationOrderForFilter=XZY)
    Child=(Type=Bone,Name="mstr_jacket_back_r")
    Parents=((Item=(Type=Bone,Name="pelvis"),Weight=1))
    Filter=@DISPATCH_RigVMDispatch_MakeStruct.Struct
    Parents.0=@n8.Struct
    Weight=@RigVMFunction_MathFloatMake.Value
   )
   > &RigUnit_ParentConstraint(
    AdvancedSettings=(InterpolationType=Average,RotationOrderForFilter=XZY)
    Child=(Type=Bone,Name="mstr_jacket_side_r")
    Parents=((Item=(Type=Bone,Name="mstr_jacket_side_r"),Weight=1))
    Filter=@DISPATCH_RigVMDispatch_MakeStruct.Struct
    Parents.0=@n8.Struct
    Weight=@RigVMFunction_MathFloatMake.Value
   )
   DISPATCH_RigVMDispatch_Constant = RigVMDispatchNode(Value=0) {template="DISPATCH_RigVMDispatch_Constant(io Value)"}
   > &RigUnit_AimConstraintLocalSpaceOffset(
    Child=(Type=Bone,Name="mstr_jacket_back_l")
    Filter=(bX=True,bY=True,bZ=True)
    Parents=((Item=(Type=Bone,Name="pelvis"),Weight=1))
    WorldUp=(Target=(X=0,Y=0,Z=1),Kind=Location,Space=(Type=Bone,Name="spine_02"))
    Weight=@DISPATCH_RigVMDispatch_Constant.Value
   )
   > &RigUnit_AimConstraintLocalSpaceOffset(
    Child=(Type=Bone,Name="mstr_jacket_back_c")
    Filter=(bX=True,bY=True,bZ=True)
    Parents=((Item=(Type=Bone,Name="pelvis"),Weight=1))
    WorldUp=(Target=(X=0,Y=0,Z=1),Kind=Location,Space=(Type=Bone,Name="spine_02"))
    Weight=@DISPATCH_RigVMDispatch_Constant.Value
   )
   > &RigUnit_AimConstraintLocalSpaceOffset(
    Child=(Type=Bone,Name="mstr_jacket_back_r")
    Filter=(bX=True,bY=True,bZ=True)
    Parents=((Item=(Type=Bone,Name="pelvis"),Weight=1))
    WorldUp=(Target=(X=0,Y=0,Z=1),Kind=Location,Space=(Type=Bone,Name="spine_02"))
    Weight=@DISPATCH_RigVMDispatch_Constant.Value
   )
   n9 = &RigVMFunction_MathQuaternionToEuler(
    RotationOrder=ZYX
    Value=&RigUnit_GetRelativeTransformForItem(Child=(Type=Bone,Name="spine_02"), Parent=(Type=Bone,Name="pelvis"), bChildInitial=False, bParentInitial=False).RelativeTransform.Rotation
   )
   > &RigUnit_ModifyTransforms(
    ItemToModify="((Item=(Type=Bone,Name=\"mstr_jacket_back_c\"),Transform=(Rotation=(X=-0.0035,Y=0.0098,Z=0.3827,W=0.9238),Translation=(X=50,Y=15,Z=0),Scale3D=(X=1,Y=1,Z=1))),(Item=(Type=Bone,Name=\"mstr_jacket_back_l\"),Transform=(Rotation=(X=-0.0283,Y=0.0544,Z=0.3823,W=0.922 ...(truncated)"
    WeightMinimum=0
    Weight=&RigVMFunction_MathDoubleRemap(SourceMaximum=-30, SourceMinimum=-10, TargetMaximum=1, TargetMinimum=0, Value=@n9.Result.Z).Result
   )
   > &RigUnit_ModifyTransforms(
    ItemToModify="((Item=(Type=Bone,Name=\"mstr_jacket_back_c\"),Transform=(Rotation=(X=0,Y=0,Z=0,W=-1),Translation=(X=-36.7035,Y=-14.0789,Z=0),Scale3D=(X=1,Y=1,Z=1))),(Item=(Type=Bone,Name=\"mstr_jacket_back_l\"),Transform=(Rotation=(X=0,Y=0,Z=0,W=-1),Translation=(X=-9.1429,Y= ...(truncated)"
    WeightMinimum=0
    Weight=&RigVMFunction_MathDoubleRemap(SourceMaximum=-53, SourceMinimum=-30, TargetMaximum=1, TargetMinimum=0, Value=@n9.Result.Z).Result
   )
   > RigVMFunction_Sequence_1 = RigVMAggregateNode()
   n10 = &RigUnit_GetRelativeTransformForItem(Child=(Type=Bone,Name="thigh_r"), Parent=(Type=Null,Name="thigh_r_null"), bChildInitial=False, bParentInitial=False)
   &RigUnit_ModifyTransforms(
    ItemToModify="((Item=(Type=Bone,Name=\"mstr_jacket_front_r\"),Transform=(Rotation=(X=0.1055,Y=0.3048,Z=-0.0886,W=0.9424),Translation=(X=6.3674,Y=4.2054,Z=4.3884),Scale3D=(X=0.715,Y=1,Z=1))),(Item=(Type=Bone,Name=\"mstr_jacket_side_r\"),Transform=(Rotation=(X=0,Y=0,Z=-0.0872 ...(truncated)"
    WeightMinimum=0
    Weight=&RigVMFunction_MathDoubleRemap(
     SourceMaximum=-80
     SourceMinimum=0
     TargetMaximum=1
     TargetMinimum=0
     Value=&RigVMFunction_MathQuaternionToEuler(RotationOrder=ZYX, Value=@n10.RelativeTransform.Rotation).Result.Z
    ).Result
    ExecutePin=@RigVMFunction_Sequence_1.B
   )
   n11 = /Script/RigVM.RigVMFunction_MathTransformToVectors(Value=@n10.RelativeTransform) {is=struct RigVMUnitNode, template="ToVectors::Execute(in Value,out Forward,out Right,out Up)"}
   > &RigUnit_ModifyTransforms(
    ItemToModify=((Item=(Type=Bone,Name="mstr_jacket_front_r"),Transform=(Rotation=(X=0.4547,Y=0.1057,Z=-0.1378,W=0.8735),Translation=(X=5.3565,Y=5.8985,Z=24.7115),Scale3D=(X=1,Y=1,Z=1))))
    Mode=OverrideLocal
    WeightMinimum=0
    Weight=&RigVMFunction_MathDoubleRemap(
     SourceMaximum=0.001
     SourceMinimum=0.8
     TargetMaximum=1
     TargetMinimum=0
     Value=/Script/RigVM.RigVMFunction_MathVectorAngle(B=(X=-315,Y=-858,Z=404), A=@n11.Forward) {is=struct RigVMUnitNode, template="AngleBetween::Execute(in A,in B,out Result)"}.Result
    ).Result
   )
   > &RigUnit_ModifyTransforms(
    ItemToModify=((Item=(Type=Bone,Name="mstr_jacket_front_r"),Transform=(Rotation=(X=0.0318,Y=0.7971,Z=-0.1728,W=0.5778),Translation=(X=13.3145,Y=10.7518,Z=23.3858),Scale3D=(X=0.3364,Y=1,Z=1))))
    Mode=OverrideLocal
    WeightMinimum=0
    Weight=&RigVMFunction_MathDoubleRemap(
     SourceMaximum=0.001
     SourceMinimum=0.5
     TargetMaximum=1
     TargetMinimum=0
     Value=&RigVMFunction_MathVectorAngle(B=(X=-242,Y=-399,Z=884), A=@n11.Forward) {template="AngleBetween::Execute(in A,in B,out Result)"}.Result
    ).Result
   )
   > &RigUnit_ModifyTransforms(
    ItemToModify=((Item=(Type=Bone,Name="mstr_jacket_front_r"),Transform=(Rotation=(X=0.465,Y=-0.1171,Z=0.0369,W=0.8768),Translation=(X=5.9686,Y=11.8745,Z=22.2973),Scale3D=(X=1,Y=1,Z=1))))
    Mode=OverrideLocal
    WeightMinimum=0
    Weight=&RigVMFunction_MathDoubleRemap(
     SourceMaximum=0.001
     SourceMinimum=0.8
     TargetMaximum=1
     TargetMinimum=0
     Value=&RigVMFunction_MathVectorAngle(B=(X=-473,Y=-874,Z=103), A=@n11.Forward) {template="AngleBetween::Execute(in A,in B,out Result)"}.Result
    ).Result
   )
   &RigUnit_ModifyTransforms(
    ItemToModify="((Item=(Type=Bone,Name=\"mstr_jacket_front_r\"),Transform=(Rotation=(X=-0.3007,Y=0,Z=0,W=-0.9537),Translation=(X=3.0439,Y=10.6062,Z=20.8983),Scale3D=(X=1,Y=1,Z=1))),(Item=(Type=Bone,Name=\"mstr_jacket_front_l\"),Transform=(Rotation=(X=0.2586,Y=0.0113,Z=0.0421, ...(truncated)"
    Mode=OverrideLocal
    WeightMinimum=0
    ExecutePin=@RigVMFunction_Sequence_1.C
    Weight=RigVMVariableNode() {variable=falling alpha}.Value
   )
   n12 = &RigUnit_GetRelativeTransformForItem(Child=(Type=Bone,Name="spine_01"), Parent=(Type=Bone,Name="pelvis"), bChildInitial=False, bParentInitial=False)
   > &RigUnit_ModifyTransforms(
    ItemToModify="((Item=(Type=Bone,Name=\"mstr_jacket_back_c\"),Transform=(Rotation=(X=0,Y=0,Z=0.2588,W=0.9659),Translation=(X=15,Y=15,Z=0),Scale3D=(X=1,Y=1,Z=1))),(Item=(Type=Bone,Name=\"mstr_jacket_back_l\"),Transform=(Rotation=(X=0,Y=0,Z=0.2588,W=0.9659),Translation=(X=-15, ...(truncated)"
    WeightMinimum=0
    Weight=&RigVMFunction_MathDoubleRemap(
     SourceMaximum=-70
     SourceMinimum=-10
     TargetMaximum=1
     TargetMinimum=0
     bClamp=False
     Value=&RigVMFunction_MathQuaternionToEuler(RotationOrder=ZYX, Value=@n12.RelativeTransform.Rotation).Result.Z
    ).Result
   )
   n13 = &RigUnit_GetRelativeTransformForItem(Child=(Type=Bone,Name="thigh_l"), Parent=(Type=Null,Name="thigh_l_null"), bChildInitial=False, bParentInitial=False)
   &RigUnit_ModifyTransforms(
    ItemToModify="((Item=(Type=Bone,Name=\"mstr_jacket_front_l\"),Transform=(Rotation=(X=-0.2758,Y=-0.019,Z=-0.0403,W=0.9602),Translation=(X=15.2128,Y=3.9687,Z=-4.419),Scale3D=(X=0.6655,Y=0.5899,Z=1.2924))),(Item=(Type=Bone,Name=\"mstr_jacket_side_l\"),Transform=(Rotation=(X=0, ...(truncated)"
    WeightMinimum=0
    Weight=&RigVMFunction_MathDoubleRemap(
     SourceMaximum=-110
     SourceMinimum=-30
     TargetMaximum=1
     TargetMinimum=0
     bClamp=False
     Value=&RigVMFunction_MathQuaternionToEuler(RotationOrder=ZYX, Value=@n13.RelativeTransform.Rotation).Result.Z
    ).Result
    ExecutePin=@RigVMFunction_Sequence_1.A
   )
   ToVectors = &RigVMFunction_MathTransformToVectors(Value=@n13.RelativeTransform) {template="ToVectors::Execute(in Value,out Forward,out Right,out Up)"}
   > &RigUnit_ModifyTransforms(
    ItemToModify="((Item=(Type=Bone,Name=\"mstr_jacket_front_l\"),Transform=(Rotation=(X=-0.1041,Y=0.6761,Z=0.1497,W=-0.7139),Translation=(X=9.5022,Y=13.7747,Z=-21.4912),Scale3D=(X=0.534,Y=1,Z=1))),(Item=(Type=Bone,Name=\"mstr_jacket_side_l\"),Transform=(Rotation=(X=0,Y=0,Z=0.0 ...(truncated)"
    Mode=OverrideLocal
    WeightMinimum=0
    Weight=&RigVMFunction_MathDoubleRemap(
     SourceMaximum=0.001
     SourceMinimum=0.6
     TargetMaximum=1
     TargetMinimum=0
     bClamp=False
     Value=&RigVMFunction_MathVectorAngle(B=(X=-230,Y=-325,Z=917), A=@ToVectors.Forward) {template="AngleBetween::Execute(in A,in B,out Result)"}.Result
    ).Result
   )
   RigVMFunction_MathVectorAngle = &RigVMFunction_MathVectorAngle(B=(X=-0.3325,Y=-0.8688,Z=0.366), A=@ToVectors.Forward) {template="AngleBetween::Execute(in A,in B,out Result)"}
   > &RigUnit_ModifyTransforms(
    ItemToModify=((Item=(Type=Bone,Name="mstr_jacket_front_l"),Transform=(Rotation=(X=-0.1605,Y=0.1082,Z=-0.2623,W=0.9454),Translation=(X=3.2204,Y=-2.1323,Z=-1.1882),Scale3D=(X=1,Y=1,Z=1))))
    WeightMinimum=0
    Weight=&RigVMFunction_MathDoubleRemap(SourceMaximum=0.001, SourceMinimum=0.3, TargetMaximum=1, TargetMinimum=0, Value=@RigVMFunction_MathVectorAngle.Result).Result
   )
   > &RigUnit_ModifyTransforms(
    ItemToModify="((Item=(Type=Bone,Name=\"mstr_jacket_front_l\"),Transform=(Rotation=(X=0,Y=0,Z=0.1305,W=0.9914),Translation=(X=3,Y=0.1955,Z=-1.0164),Scale3D=(X=0.7,Y=1,Z=1))),(Item=(Type=Bone,Name=\"mstr_jacket_side_l\"),Transform=(Rotation=(X=0,Y=0,Z=0.0872,W=0.9962),Transla ...(truncated)"
    WeightMinimum=0
    Weight=&RigVMFunction_MathDoubleRemap(
     SourceMaximum=0.001
     SourceMinimum=0.7
     TargetMaximum=1
     TargetMinimum=0
     Value=&RigVMFunction_MathVectorAngle(B=(X=0.8195,Y=-0.561,Z=-0.117), A=@ToVectors.Forward) {template="AngleBetween::Execute(in A,in B,out Result)"}.Result
    ).Result
   )
   > &RigUnit_ModifyTransforms(
    ItemToModify=((Item=(Type=Bone,Name="mstr_jacket_front_l"),Transform=(Rotation=(X=0,Y=0,Z=-0.0436,W=0.999),Translation=(X=-1.4439,Y=2.8219,Z=-0.395),Scale3D=(X=0.8901,Y=1,Z=1.2085))))
    WeightMinimum=0
    Weight=&RigVMFunction_MathDoubleRemap(
     SourceMaximum=0.001
     SourceMinimum=0.5
     TargetMaximum=1
     TargetMinimum=0
     Value=&RigVMFunction_MathVectorAngle(B=(X=187906,Y=-915588,Z=355551), A=@ToVectors.Forward) {template="AngleBetween::Execute(in A,in B,out Result)"}.Result
    ).Result
   )
   > &RigUnit_ModifyTransforms(
    ItemToModify=((Item=(Type=Bone,Name="mstr_jacket_front_l"),Transform=(Rotation=(X=-0.0916,Y=0.0101,Z=0.1628,W=0.9824),Translation=(X=11.2504,Y=5.4708,Z=-16.382),Scale3D=(X=0.7985,Y=0.6066,Z=0.8208))))
    Mode=OverrideLocal
    WeightMinimum=0
    Weight=&RigVMFunction_MathDoubleRemap(
     SourceMaximum=0.001
     SourceMinimum=0.4
     TargetMaximum=1
     TargetMinimum=0
     Value=&RigVMFunction_MathVectorAngle(B=(X=-466,Y=-778,Z=419), A=@ToVectors.Forward) {template="AngleBetween::Execute(in A,in B,out Result)"}.Result
    ).Result
   )
   RigUnit_GetTransform = &RigUnit_GetTransform(Item=(Type=Bone,Name="spine_02"), Space=LocalSpace, bInitial=False)
   n14 = &RigVMFunction_MathQuaternionToEuler(RotationOrder=ZYX, Value=@RigUnit_GetTransform.Transform.Rotation)
   Multiply = &RigVMFunction_MathDoubleMul(
    B=&RigVMFunction_MathDoubleRemap(SourceMaximum=22, SourceMinimum=5, TargetMaximum=1, TargetMinimum=0, Value=@n14.Result.Z).Result
    A=&RigVMFunction_MathDoubleRemap(SourceMaximum=5, SourceMinimum=0, TargetMaximum=1, TargetMinimum=0, Value=@n14.Result.Y).Result
   )
   > &RigUnit_ModifyTransforms(
    ItemToModify=((Item=(Type=Bone,Name="mstr_jacket_front_l"),Transform=(Rotation=(X=0,Y=0,Z=0,W=1),Translation=(X=-2.3412,Y=-0.2558,Z=-3.5506),Scale3D=(X=0.573,Y=0.9761,Z=1))))
    WeightMinimum=0
    Weight=@Multiply.Result
   )
   > &RigUnit_ModifyTransforms(
    ItemToModify=((Item=(Type=Bone,Name="mstr_jacket_front_l"),Transform=(Rotation=(X=0,Y=0,Z=0,W=1),Translation=(X=4.233,Y=0,Z=-1.3369),Scale3D=(X=1,Y=1,Z=1))))
    WeightMinimum=0
    Weight=&RigVMFunction_MathDoubleRemap(
     SourceMaximum=0.001
     SourceMinimum=0.4
     TargetMaximum=1
     TargetMinimum=0
     Value=&RigVMFunction_MathVectorAngle(B=(X=492,Y=-662,Z=564), A=@ToVectors.Forward) {template="AngleBetween::Execute(in A,in B,out Result)"}.Result
    ).Result
   )
   > &RigUnit_ModifyTransforms(
    ItemToModify=((Item=(Type=Bone,Name="mstr_jacket_front_l"),Transform=(Rotation=(X=0,Y=0.0872,Z=0,W=0.9962),Translation=(X=2.1828,Y=0,Z=0),Scale3D=(X=1,Y=1,Z=1))))
    WeightMinimum=0
    Weight=&RigVMFunction_MathDoubleRemap(
     SourceMaximum=0.001
     SourceMinimum=0.5
     TargetMaximum=1
     TargetMinimum=0
     Value=&RigVMFunction_MathVectorAngle(B=(X=768,Y=-550,Z=326), A=@ToVectors.Forward) {template="AngleBetween::Execute(in A,in B,out Result)"}.Result
    ).Result
   )
   &RigUnit_ParentConstraint(
    AdvancedSettings=(InterpolationType=Average,RotationOrderForFilter=XZY)
    Child=(Type=Bone,Name="mstr_shorts_back_l")
    Parents=((Item=(Type=Bone,Name="thigh_twist_01_l"),Weight=1))
    Weight=1
    ExecutePin=@RigVMFunction_Sequence.B
   )
   > &RigUnit_ParentConstraint(
    AdvancedSettings=(InterpolationType=Average,RotationOrderForFilter=XZY)
    Child=(Type=Bone,Name="mstr_shorts_back_r")
    Parents=((Item=(Type=Bone,Name="thigh_twist_01_r"),Weight=1))
    Weight=1
   )
   n15 = &RigUnit_GetRelativeTransformForItem(Child=(Type=Bone,Name="calf_r"), Parent=(Type=Null,Name="calf_r_null"), bChildInitial=False, bParentInitial=False)
   > &RigUnit_ModifyTransforms(
    ItemToModify="((Item=(Type=Bone,Name=\"mstr_shorts_back_r\"),Transform=(Rotation=(X=0,Y=-0.1391,Z=0,W=-0.9903),Translation=(X=-9.2892,Y=0,Z=0),Scale3D=(X=0.6348,Y=1,Z=1))),(Item=(Type=Bone,Name=\"mstr_shorts_inner_r\"),Transform=(Rotation=(X=0,Y=0,Z=-0.1406,W=0.9901),Transl ...(truncated)"
    WeightMinimum=0
    Weight=&RigVMFunction_MathDoubleRemap(
     SourceMaximum=130
     SourceMinimum=0
     TargetMaximum=1
     TargetMinimum=0
     bClamp=False
     Value=&RigVMFunction_MathQuaternionToEuler(RotationOrder=ZYX, Value=@n15.RelativeTransform.Rotation).Result.Z
    ).Result
   )
   n16 = &RigUnit_GetRelativeTransformForItem(Child=(Type=Bone,Name="calf_l"), Parent=(Type=Null,Name="calf_l_null"), bChildInitial=False, bParentInitial=False)
   > &RigUnit_ModifyTransforms(
    ItemToModify="((Item=(Type=Bone,Name=\"mstr_shorts_back_l\"),Transform=(Rotation=(X=0,Y=-0.1391,Z=0,W=-0.9903),Translation=(X=9.2892,Y=0,Z=0),Scale3D=(X=0.6348,Y=1,Z=1))),(Item=(Type=Bone,Name=\"mstr_shorts_inner_l\"),Transform=(Rotation=(X=0,Y=0,Z=-0.1406,W=0.9901),Transla ...(truncated)"
    WeightMinimum=0
    Weight=&RigVMFunction_MathDoubleRemap(
     SourceMaximum=130
     SourceMinimum=0
     TargetMaximum=1
     TargetMinimum=0
     bClamp=False
     Value=&RigVMFunction_MathQuaternionToEuler(RotationOrder=ZYX, Value=@n16.RelativeTransform.Rotation).Result.Z
    ).Result
   )
  ]}
]
hierarchy [
 root {loc=<0 0 0>, rot=<0 0 0 1>}
 pelvis {parent=root, loc=<0 -0.7321 92.4218>, rot=<0.0127 0.707 -0.0127 -0.707>}
 spine_01 {parent=pelvis, loc=<5.6002 -0.5128 0>, rot=<0 0 0.034 -0.9994>}
 spine_02 {parent=spine_01, loc=<8.0753 0.0139 0>, rot=<0 0 0.0238 -0.9997>}
 spine_03 {parent=spine_02, loc=<7.0393 0 0>, rot=<0 0 0.0149 -0.9999>}
 spine_04 {parent=spine_03, loc=<8.1024 0 0>, rot=<0 0 -0.0132 -0.9999>}
 spine_05 {parent=spine_04, loc=<10.0146 0 0>, rot=<0 0 -0.0015 -1>}
 neck_01 {parent=spine_05, loc=<10.815 1.9196 0>, rot=<0 0 -0.1864 -0.9825>}
 neck_02 {parent=neck_01, loc=<4.7347 0.3045 0>, rot=<0 0 0.0363 -0.9993>}
 head {parent=neck_02, loc=<5.7632 0.5552 0>, rot=<0 0 0.115 -0.9934>}
 mstr_neck_01 {parent=neck_01, loc=<0 0 0>, rot=<0 0 0 1>}
 attach_backpack {parent=spine_05, loc=<-0.123 -15.1423 0>, rot=<-0.0093 -0.707 0.0093 -0.707>}
 clavicle_l {parent=spine_05, loc=<2.5856 -0.2866 -4.3777>, rot=<0.0465 -0.7289 -0.009 -0.6829>}
 upperarm_l {parent=clavicle_l, loc=<18.2264 0.1342 -0.0011>, rot=<0.0253 -0.2693 0.0013 -0.9627>}
 lowerarm_l {parent=upperarm_l, loc=<29.0071 0.1686 -1.239>, rot=<0.0391 -0.0099 -0.2453 -0.9686>}
 hand_l {parent=lowerarm_l, loc=<25.8607 0.222 1.0274>, rot=<-0.7193 -0.0988 0.1193 -0.6773>}
 thumb_01_l {parent=hand_l, loc=<1.3118 -0.3191 -2.459>, rot=<0.6546 -0.4042 0.0664 -0.6354>}
 thumb_02_l {parent=thumb_01_l, loc=<5.9336 -0.0448 0.0401>, rot=<-0.0037 0.0213 0.298 -0.9543>}
 thumb_03_l {parent=thumb_02_l, loc=<2.2223 0.3682 -0.0907>, rot=<0.0118 -0.013 0.0066 -0.9998>}
 index_metacarpal_l {parent=hand_l, loc=<1.6535 0.0183 -2.96>, rot=<0.0158 -0.0124 -0.0173 -0.9996>}
 index_01_l {parent=index_metacarpal_l, loc=<8.3008 0.0357 -0.0524>, rot=<-0.0044 0.0485 0.0924 -0.9945>}
 index_02_l {parent=index_01_l, loc=<3.3587 -0.3004 0.0452>, rot=<-0.002 0.0027 0.3225 -0.9466>}
 index_03_l {parent=index_02_l, loc=<2.5578 -0.123 -0.0252>, rot=<-0.002 0.0067 0.2558 -0.9667>}
 middle_metacarpal_l {parent=hand_l, loc=<2.1856 -0.1725 -0.5984>, rot=<0.006 0.0168 -0.0329 -0.9993>}
 middle_01_l {parent=middle_metacarpal_l, loc=<7.3275 -0.2447 0.0129>, rot=<-0.0208 0.0042 0.0858 -0.9961>}
 middle_02_l {parent=middle_01_l, loc=<4.0123 0.1401 0.1175>, rot=<-0.0218 0.0258 0.3512 -0.9357>}
 middle_03_l {parent=middle_02_l, loc=<2.6077 0.0534 0.1373>, rot=<0.0082 0.0188 0.1708 -0.9851>}
 ring_metacarpal_l {parent=hand_l, loc=<2.2074 -0.3625 1.3638>, rot=<-0.0284 0.0722 -0.0458 -0.9959>}
 ring_01_l {parent=ring_metacarpal_l, loc=<7.0295 -0.6197 -0.1446>, rot=<-0.0489 -0.0252 0.1892 -0.9804>}
 ring_02_l {parent=ring_01_l, loc=<3.6174 0.4647 -0.1027>, rot=<0 0 0.2806 -0.9598>}
 ring_03_l {parent=ring_02_l, loc=<2.4152 0.114 0.0199>, rot=<0 0 0.1189 -0.9929>}
 pinky_metacarpal_l {parent=hand_l, loc=<1.7616 -0.9753 3.3814>, rot=<-0.0397 0.0814 -0.069 -0.9935>}
 pinky_01_l {parent=pinky_metacarpal_l, loc=<7.1709 -0.4831 -0.1094>, rot=<-0.0654 -0.0035 0.2242 -0.9723>}
 pinky_02_l {parent=pinky_01_l, loc=<2.8274 0.0893 -0.1043>, rot=<0.0037 0.0011 0.2854 -0.9584>}
 pinky_03_l {parent=pinky_02_l, loc=<2.2046 0.0307 -0.0078>, rot=<-0.0002 0.0098 0.1445 -0.9895>}
 deform_wrist_dn_base_l {parent=hand_l, loc=<0 -3.4173 0>, rot=<0 0 0 1>}
 deform_wrist_dn_l {parent=deform_wrist_dn_base_l, loc=<0 0 0>, rot=<0 0 0 1>}
 deform_wrist_up_base_l {parent=hand_l, loc=<0 2.7805 0>, rot=<0 0 0 1>}
 deform_wrist_up_l {parent=deform_wrist_up_base_l, loc=<0 0 0>, rot=<0 0 0 1>}
 hand_sleeve_l {parent=hand_l, loc=<0.0001 0 0>, rot=<0 0 0 1>}
 lowerarm_twist_02_l {parent=lowerarm_l, loc=<9.6199 0 0>, rot=<0 0 0 -1>}
 lowerarm_twist_01_l {parent=lowerarm_l, loc=<17.8549 0 0>, rot=<0 0 0 -1>}
 deform_elbow_in_base_l {parent=lowerarm_l, loc=<0.2112 5.645 0.8274>, rot=<0 0 0 1>}
 deform_elbow_in_l {parent=deform_elbow_in_base_l, loc=<0 0 0>, rot=<0 0 0 1>}
 deform_elbow_out_base_l {parent=lowerarm_l, loc=<0 -7.5129 0>, rot=<0 0 0 1>}
 deform_elbow_out_l {parent=deform_elbow_out_base_l, loc=<0 0 0>, rot=<0 0 0 1>}
 hand_l_null {parent=lowerarm_l, type=Null, loc=<25.8607 0.222 1.0274>, rot=<-0.7193 -0.0988 0.1193 -0.6773>}
 upperarm_twist_01_l {parent=upperarm_l, loc=<7.9796 0.0843 -0.6195>, rot=<0 0 0 1>}
 deform_upperarm_bk_base_l {parent=upperarm_twist_01_l, loc=<-6.8983 -13.3826 -1.3258>, rot=<0 0 0 1>}
 deform_upperarm_bk_l {parent=deform_upperarm_bk_base_l, loc=<0 0 0>, rot=<0 0 0 1>}
 deform_upperarm_fr_base_l {parent=upperarm_twist_01_l, loc=<-8.3487 10.7117 -2.1455>, rot=<0 0 0 1>}
 deform_upperarm_fr_l {parent=deform_upperarm_fr_base_l, loc=<0 0 0>, rot=<0 0 0 1>}
 deform_upperarm_up_base_l {parent=upperarm_twist_01_l, loc=<-11.2012 -1.2234 8.7784>, rot=<0 0 0 1>}
 deform_upperarm_up_l {parent=deform_upperarm_up_base_l, loc=<0 0 0>, rot=<0 0 0 1>}
 upperarm_twist_02_l {parent=upperarm_l, loc=<19.5059 0.0843 -0.6195>, rot=<0 0 0 1>}
 attach_cape {parent=spine_05, loc=<9.1509 -13.3783 0>, rot=<0.0616 -0.7044 -0.0616 -0.7044>}
 clavicle_r {parent=spine_05, loc=<2.5855 -0.2866 4.3777>, rot=<0.7289 0.0465 -0.6829 0.009>}
 upperarm_r {parent=clavicle_r, loc=<-18.2264 -0.1342 0.001>, rot=<0.0253 -0.2693 0.0013 -0.9627>}
 lowerarm_r {parent=upperarm_r, loc=<-29.0071 -0.1686 1.2389>, rot=<0.0391 -0.0099 -0.2453 -0.9686>}
 hand_r {parent=lowerarm_r, loc=<-25.8607 -0.222 -1.0274>, rot=<-0.7193 -0.0988 0.1193 -0.6773>}
 thumb_01_r {parent=hand_r, loc=<-1.3118 0.3191 2.459>, rot=<0.6546 -0.4042 0.0664 -0.6354>}
 thumb_02_r {parent=thumb_01_r, loc=<-5.9336 0.0448 -0.0401>, rot=<-0.0037 0.0213 0.298 -0.9543>}
 thumb_03_r {parent=thumb_02_r, loc=<-2.2223 -0.3682 0.0907>, rot=<0.0118 -0.013 0.0066 -0.9998>}
 index_metacarpal_r {parent=hand_r, loc=<-1.6535 -0.0183 2.96>, rot=<0.0158 -0.0124 -0.0173 -0.9996>}
 index_01_r {parent=index_metacarpal_r, loc=<-8.3008 -0.0357 0.0524>, rot=<-0.0044 0.0485 0.0924 -0.9945>}
 index_02_r {parent=index_01_r, loc=<-3.3587 0.3004 -0.0452>, rot=<-0.002 0.0027 0.3225 -0.9466>}
 index_03_r {parent=index_02_r, loc=<-2.5578 0.123 0.0252>, rot=<-0.002 0.0067 0.2558 -0.9667>}
 middle_metacarpal_r {parent=hand_r, loc=<-2.1856 0.1725 0.5984>, rot=<0.006 0.0168 -0.0329 -0.9993>}
 middle_01_r {parent=middle_metacarpal_r, loc=<-7.3275 0.2447 -0.0129>, rot=<-0.0208 0.0042 0.0858 -0.9961>}
 middle_02_r {parent=middle_01_r, loc=<-4.0123 -0.1401 -0.1175>, rot=<-0.0218 0.0258 0.3512 -0.9357>}
 middle_03_r {parent=middle_02_r, loc=<-2.6077 -0.0534 -0.1373>, rot=<0.0082 0.0188 0.1708 -0.9851>}
 ring_metacarpal_r {parent=hand_r, loc=<-2.2074 0.3625 -1.3638>, rot=<-0.0284 0.0722 -0.0458 -0.9959>}
 ring_01_r {parent=ring_metacarpal_r, loc=<-7.0295 0.6197 0.1446>, rot=<-0.0489 -0.0252 0.1892 -0.9804>}
 ring_02_r {parent=ring_01_r, loc=<-3.6174 -0.4647 0.1027>, rot=<0 0 0.2806 -0.9598>}
 ring_03_r {parent=ring_02_r, loc=<-2.4152 -0.114 -0.0199>, rot=<0 0 0.1189 -0.9929>}
 pinky_metacarpal_r {parent=hand_r, loc=<-1.7616 0.9753 -3.3814>, rot=<-0.0397 0.0814 -0.069 -0.9935>}
 pinky_01_r {parent=pinky_metacarpal_r, loc=<-7.1709 0.4831 0.1094>, rot=<-0.0654 -0.0035 0.2242 -0.9723>}
 pinky_02_r {parent=pinky_01_r, loc=<-2.8274 -0.0893 0.1043>, rot=<0.0037 0.0011 0.2854 -0.9584>}
 pinky_03_r {parent=pinky_02_r, loc=<-2.2046 -0.0307 0.0078>, rot=<-0.0002 0.0098 0.1445 -0.9895>}
 deform_wrist_dn_base_r {parent=hand_r, loc=<0 3.4172 0>, rot=<0 0 0 -1>}
 deform_wrist_dn_r {parent=deform_wrist_dn_base_r, loc=<0 0 0>, rot=<0 0 0 -1>}
 deform_wrist_up_base_r {parent=hand_r, loc=<-0.0001 -2.7804 0>, rot=<0 0 0 -1>}
 deform_wrist_up_r {parent=deform_wrist_up_base_r, loc=<0 0 0>, rot=<0 0 0 -1>}
 hand_sleeve_r {parent=hand_r, loc=<0 0 0>, rot=<0 0 0 1>}
 lowerarm_twist_02_r {parent=lowerarm_r, loc=<-9.6199 0 0>, rot=<0 0 0 -1>}
 lowerarm_twist_01_r {parent=lowerarm_r, loc=<-17.8549 0 0>, rot=<0 0 0 -1>}
 deform_elbow_in_base_r {parent=lowerarm_r, loc=<-0.2114 -5.6449 -0.827>, rot=<0 0 0 1>}
 deform_elbow_in_r {parent=deform_elbow_in_base_r, loc=<0 0 0>, rot=<0 0 0 1>}
 deform_elbow_out_base_r {parent=lowerarm_r, loc=<-0.0002 7.5129 0.0002>, rot=<0 0 0 1>}
 deform_elbow_out_r {parent=deform_elbow_out_base_r, loc=<0 0 0>, rot=<0 0 0 1>}
 upperarm_twist_01_r {parent=upperarm_r, loc=<-7.9796 -0.0843 0.6195>, rot=<0 0 0 1>}
 deform_upperarm_bk_base_r {parent=upperarm_twist_01_r, loc=<6.8984 13.3826 1.3257>, rot=<0 0 0 1>}
 deform_upperarm_bk_r {parent=deform_upperarm_bk_base_r, loc=<0 0 0>, rot=<0 0 0 1>}
 deform_upperarm_fr_base_r {parent=upperarm_twist_01_r, loc=<8.3485 -10.7117 2.1457>, rot=<0 0 0 1>}
 deform_upperarm_fr_r {parent=deform_upperarm_fr_base_r, loc=<0 0 0>, rot=<0 0 0 1>}
 deform_upperarm_up_base_r {parent=upperarm_twist_01_r, loc=<11.2014 1.2234 -8.7787>, rot=<0 0 0 1>}
 deform_upperarm_up_r {parent=deform_upperarm_up_base_r, loc=<0 0 0>, rot=<0 0 0 1>}
 upperarm_twist_02_r {parent=upperarm_r, loc=<-19.5059 -0.0843 0.6195>, rot=<0 0 0 1>}
 spine_thigh_r {parent=spine_03, loc=<0 0 0>, rot=<0 0 -0.9984 0.0572>}
 mstr_jacket_back_c {parent=spine_02, loc=<10.59 -14.3348 0>, rot=<0 0 0.9992 -0.0399>}
 dyn_jacket_back_c {parent=mstr_jacket_back_c, loc=<0 0 0>, rot=<0 0 0 1>}
 mstr_jacket_side_l {parent=spine_02, loc=<6.9511 -1.6978 -19.0854>, rot=<-0.7111 -0.0229 -0.0486 -0.701>}
 dyn_jacket_side_l {parent=mstr_jacket_side_l, loc=<0 0 0>, rot=<0 0 0 1>}
 mstr_jacket_back_l {parent=spine_02, loc=<9.2804 -13.2546 -12.7723>, rot=<-0.3514 -0.014 -0.0373 -0.9354>}
 dyn_jacket_back_l {parent=mstr_jacket_back_l, loc=<0 0 0>, rot=<0 0 0 1>}
 mstr_jacket_back_r {parent=spine_02, loc=<9.2803 -13.2545 12.7723>, rot=<-0.014 0.3514 0.9354 -0.0373>}
 dyn_jacket_back_r {parent=mstr_jacket_back_r, loc=<0 0 0>, rot=<0 0 0 1>}
 mstr_jacket_side_r {parent=spine_02, loc=<6.9509 -1.6978 19.0854>, rot=<0.0229 -0.7111 -0.701 0.0486>}
 dyn_jacket_side_r {parent=mstr_jacket_side_r, loc=<0 0 0>, rot=<0 0 0 1>}
 thigh_l {parent=pelvis, loc=<-1.8716 -0.0109 -11.9936>, rot=<0.0012 0.0572 0.0155 -0.9982>}
 calf_l {parent=thigh_l, loc=<-41.9178 -0.2246 0.0925>, rot=<0.0006 -0.0161 -0.0391 -0.9991>}
 foot_l {parent=calf_l, loc=<-38.7057 1.2439 0.2452>, rot=<0.0975 -0.0474 0.0317 -0.9936>}
 ball_l {parent=foot_l, loc=<-6.1614 13.1469 0.6807>, rot=<0.0001 -0.0001 -0.7063 -0.7079>}
 mstr_shoe_tongue_l {parent=foot_l, loc=<2.7198 3.6294 1.1174>, rot=<-0.0371 -0.054 0.9958 -0.0637>}
 mstr_foot_l {parent=foot_l, loc=<0 0 0>, rot=<0 0 0 1>}
 calf_twist_02_l {parent=calf_l, loc=<-15.9268 -0.0967 0.0397>, rot=<0.0237 0.0095 -0.0035 -0.9997>}
 calf_twist_01_l {parent=calf_l, loc=<-32.9334 -0.0967 0.0397>, rot=<0.0237 0.0095 -0.0035 -0.9997>}
 deform_knee_in_base_l {parent=calf_l, loc=<1.298 -8.0356 -0.0496>, rot=<0 0 0 1>}
 deform_knee_in_l {parent=deform_knee_in_base_l, loc=<0 0 0>, rot=<0 0 0 1>}
 deform_knee_out_base_l {parent=calf_l, loc=<0 6.8739 0>, rot=<0 0 0 1>}
 deform_knee_out_l {parent=deform_knee_out_base_l, loc=<0 0 0>, rot=<0 0 0 1>}
 foot_l_null {parent=calf_l, type=Null, loc=<-38.7057 1.2438 0.2452>, rot=<0.0975 -0.0474 0.0317 -0.9936>}
 thigh_twist_01_l {parent=thigh_l, loc=<-21.6303 0 0>, rot=<0.0013 0 0 -1>}
 mstr_shorts_pocket_l {parent=thigh_twist_01_l, loc=<12.364 -2.1574 -8.1452>, rot=<0.9984 -0.0073 0.0451 0.0321>}
 dyn_shortsElement_B_l {parent=mstr_shorts_pocket_l, loc=<-11.8547 -4.9963 0.8547>, rot=<-0.0326 0.0124 -0.0168 -0.9992>}
 mstr_shorts_inner_l {parent=thigh_twist_01_l, loc=<-1.111 1.5605 11.7277>, rot=<0.9991 -0.0029 -0.0174 0.0395>}
 dyn_shorts_inner_l {parent=mstr_shorts_inner_l, loc=<0 0 0>, rot=<0 0 0 1>}
 mstr_shorts_front_l {parent=thigh_twist_01_l, loc=<-1.3224 10.1892 0.3423>, rot=<-0.6765 0 0 -0.7364>}
 dyn_shorts_front_l {parent=mstr_shorts_front_l, loc=<0 0 0>, rot=<0 0 0 1>}
 mstr_shorts_out_l {parent=thigh_twist_01_l, loc=<-7.1402 -2.0243 -10.1086>, rot=<0 0 0 1>}
 dyn_shorts_out_l {parent=mstr_shorts_out_l, loc=<0 0 0>, rot=<0 0 0 1>}
 dyn_shortsElement_out_l {parent=dyn_shorts_out_l, loc=<-7.5735 0 -0.8454>, rot=<0.0145 0.0337 -0.0126 -0.9992>}
 mstr_shorts_back_l {parent=thigh_l, loc=<-21.1908 -12.7632 2.02>, rot=<0.7119 0.0056 0.0145 -0.7021>}
 dyn_shorts_back_l {parent=mstr_shorts_back_l, loc=<0 0 0>, rot=<0 0 0 1>}
 deform_groin_base_l {parent=thigh_l, loc=<-2.1907 8.5661 -1.7442>, rot=<0 0 0 -1>}
 deform_groin_l {parent=deform_groin_base_l, loc=<0 0 0>, rot=<0 0 0 1>}
 calf_l_null {parent=thigh_l, type=Null, loc=<-41.9178 -0.2246 0.0926>, rot=<0.0006 -0.0161 -0.0391 -0.9991>}
 jacket_target_l {parent=thigh_l, loc=<-25 10 -10>, rot=<0 0 0 1>}
 thigh_r {parent=pelvis, loc=<-1.8716 -0.0109 11.9936>, rot=<0.0572 -0.0012 0.9982 0.0155>}
 calf_r {parent=thigh_r, loc=<41.9178 0.2246 -0.0925>, rot=<0.0006 -0.0161 -0.0391 -0.9991>}
 foot_r {parent=calf_r, loc=<38.7057 -1.2438 -0.2452>, rot=<0.0975 -0.0474 0.0317 -0.9936>}
 ball_r {parent=foot_r, loc=<6.1614 -13.1469 -0.6807>, rot=<0.0001 -0.0001 -0.7063 -0.7079>}
 mstr_shoe_tongue_r {parent=foot_r, loc=<-2.7198 -3.6294 -1.1173>, rot=<-0.0371 -0.054 0.9958 -0.0637>}
 mstr_foot_r {parent=foot_r, loc=<0 0 0>, rot=<0 0 0 1>}
 calf_twist_02_r {parent=calf_r, loc=<15.9268 0.0967 -0.0397>, rot=<0.0237 0.0095 -0.0035 -0.9997>}
 calf_twist_01_r {parent=calf_r, loc=<32.9336 0.0967 -0.0397>, rot=<0.0237 0.0095 -0.0035 -0.9997>}
 deform_knee_in_base_r {parent=calf_r, loc=<-1.298 8.0356 0.0496>, rot=<0 0 0 1>}
 deform_knee_in_r {parent=deform_knee_in_base_r, loc=<0 0 0>, rot=<0 0 0 1>}
 deform_knee_out_base_r {parent=calf_r, loc=<0.0002 -6.8739 0>, rot=<0 0 0 1>}
 deform_knee_out_r {parent=deform_knee_out_base_r, loc=<0 0 0>, rot=<0 0 0 1>}
 foot_r_null {parent=calf_r, type=Null, loc=<38.7057 -1.2438 -0.2453>, rot=<0.0975 -0.0474 0.0317 -0.9936>}
 thigh_twist_01_r {parent=thigh_r, loc=<21.6303 0 0>, rot=<0.0013 0 0 -1>}
 mstr_shorts_pocket_r {parent=thigh_twist_01_r, loc=<-12.364 2.1574 8.1453>, rot=<0.9984 -0.0073 0.0451 0.0321>}
 dyn_shortsElement_B_r {parent=mstr_shorts_pocket_r, loc=<11.8547 4.9963 -0.8546>, rot=<-0.0331 0.0112 0.0193 -0.9992>}
 dyn_shortsElement_B1_r {parent=dyn_shortsElement_B_r, loc=<3.049 0.2277 0.3745>, rot=<0 0 0 1>}
 dyn_shortsElement_B1a_r {parent=dyn_shortsElement_B1_r, loc=<1.5553 0.1707 0.1828>, rot=<-0.2151 0.0715 -0.0159 -0.9738>}
 dyn_shortsElement_B2_r {parent=dyn_shortsElement_B1a_r, loc=<1.9151 0 0>, rot=<0 0 0 1>}
 mstr_shorts_front_r {parent=thigh_twist_01_r, loc=<1.3224 -10.1892 -0.3423>, rot=<-0.6765 0 0 -0.7364>}
 dyn_shorts_front_r {parent=mstr_shorts_front_r, loc=<0 0 0>, rot=<0 0 0 1>}
 mstr_shorts_inner_r {parent=thigh_twist_01_r, loc=<1.111 -1.5605 -11.7277>, rot=<0.9991 -0.0029 -0.0174 0.0395>}
 dyn_shorts_inner_r {parent=mstr_shorts_inner_r, loc=<0 0 0>, rot=<0 0 0 1>}
 mstr_shorts_out_r {parent=thigh_twist_01_r, loc=<7.1402 2.0243 10.1086>, rot=<0 0 0 1>}
 dyn_shorts_out_r {parent=mstr_shorts_out_r, loc=<0 0 0>, rot=<0 0 0 1>}
 dyn_shortsElement_out_r {parent=dyn_shorts_out_r, loc=<7.5736 0 0.8455>, rot=<0.0145 0.0337 -0.0126 -0.9992>}
 mstr_shorts_back_r {parent=thigh_r, loc=<21.1907 12.7632 -2.02>, rot=<0.7119 0.0056 0.0145 -0.7021>}
 dyn_shorts_back_r {parent=mstr_shorts_back_r, loc=<0 0 0>, rot=<0 0 0 1>}
 deform_groin_base_r {parent=thigh_r, loc=<2.1909 -8.5661 1.7442>, rot=<0 0 0 1>}
 deform_groin_r {parent=deform_groin_base_r, loc=<0 0 0>, rot=<0 0 0 1>}
 calf_r_null {parent=thigh_r, type=Null, loc=<41.9178 0.2246 -0.0926>, rot=<0.0006 -0.0161 -0.0391 -0.9991>}
 jacket_target_r {parent=thigh_r, loc=<25 -10 10>, rot=<0 0 0 1>}
 deform_glute_base_l {parent=pelvis, loc=<-10.1877 -11.6218 -12.899>, rot=<0.0275 0.0501 -0.4472 -0.8926>}
 deform_glute_l {parent=deform_glute_base_l, loc=<0 0 0>, rot=<0 0 0 1>}
 deform_glute_base_r {parent=pelvis, loc=<-10.1877 -11.6218 12.899>, rot=<0.0501 -0.0275 0.8926 -0.4472>}
 deform_glute_r {parent=deform_glute_base_r, loc=<0 0 0>, rot=<0 0 0 1>}
 thigh_l_null {parent=pelvis, type=Null, loc=<-1.8716 -0.0109 -11.9936>, rot=<0.0012 0.0572 0.0155 -0.9982>}
 thigh_r_null {parent=pelvis, type=Null, loc=<-1.8716 -0.0109 11.9936>, rot=<0.0572 -0.0012 0.9982 0.0155>}
 mstr_buckle {parent=pelvis, loc=<3.9239 12.0428 0>, rot=<0 0 0 1>}
 jacket_targetSide_l {parent=pelvis, loc=<-15 0 -20>, rot=<0 0 0 1>}
 jacket_targetSide_r {parent=pelvis, loc=<-15 0 20>, rot=<0 0 0 1>}
 mstr_jacket_front_r {parent=pelvis, loc=<-3.3232 10.6062 13.4355>, rot=<0 0 0 1>}
 dyn_jacket_front_r {parent=mstr_jacket_front_r, loc=<0 0 0>, rot=<0 0 0 1>}
 mstr_jacket_front_l {parent=pelvis, loc=<-3.3232 10.6062 -13.4355>, rot=<0 0 0 1>}
 dyn_jacket_front_l {parent=mstr_jacket_front_l, loc=<0 0 0>, rot=<0 0 0 1>}
 WorldSpace {type=Reference, loc=<0 0 0>, rot=<0 0 0 1>}
]
variables [
 RBAN {type=bool, default=True}
 falling alpha {type=double, default=0}
]
