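# Physics asset for Jacket1 on the large male WarmShadeWeasel cosmetic body, covering the front and side jacket bones on both sides.
# Each dyn_jacket capsule is constrained to a small mstr_jacket sphere, alongside larger capsules approximating pelvis, spine_02 and thighs.
# Side flaps swing with twist locked, while front flaps mainly twist, with asymmetric limits: right swing is locked, left slightly limited.
# Thighs constrain to the pelvis, and thigh twist bodies rotate freely on the thighs.
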PhysicsAsset /BRCosmetics/Characters/Player/Male/Large/Bodies/M_LRG_WarmShadeWeasel/Meshes/M_LRG_WarmShadeWeasel_Physics_Jacket1
bodies {
 dyn_jacket_front_l [{
   shapeType=Capsule
   length=13.1572
   radius=4.1377
   shapeName=dyn_jacket_front_r_capsule
   center=<4.6058 -3.8214 -4.3839>
   rotation=<rot 12.5056 1.6766 -124.9435>
  }]
 dyn_jacket_front_r [{
   shapeType=Capsule
   length=11.0905
   radius=4.1377
   shapeName=dyn_jacket_front_r_capsule
   center=<5.0878 -2.629 4.4166>
   rotation=<rot -11.988 10.104 -57.5062>
  }]
 dyn_jacket_side_l [{
   shapeType=Capsule
   length=7.1173
   radius=6.3338
   shapeName=dyn_jacketA_front_r_capsule
   center=<-12.0309 -4.4867 0>
   rotation=<rot -90 180 -7.7119>
  }]
 dyn_jacket_side_r [{
   shapeType=Capsule
   length=7.1173
   radius=6.3338
   shapeName=dyn_jacketA_front_r_capsule
   center=<11.9501 4.0077 0>
   rotation=<rot -90 180 172.2441>
  }]
 mstr_jacket_front_l [{shapeType=Sphere, radius=2.0807, shapeName=mstr_jacket_front_l_sphere}]
 mstr_jacket_front_r [{shapeType=Sphere, radius=2.0807, shapeName=mstr_jacket_front_l_sphere}]
 mstr_jacket_side_l [{shapeType=Sphere, radius=2.0315, shapeName=mstr_jacket_side_r_sphere}]
 mstr_jacket_side_r [{shapeType=Sphere, radius=2.0315, shapeName=mstr_jacket_side_r_sphere}]
 pelvis [{shapeType=Capsule, length=8.8103, radius=13.5027, shapeName=pelvis_capsule, center=<0 -2.3127 0>, rotation=<rot 0 -0.0002 -4.9122>}]
 spine_02 [{shapeType=Capsule, length=10.0796, radius=13.0749, shapeName=spine_02_capsule, center=<0 -1.7401 0>, rotation=<rot 0 179.999 -180>}]
 thigh_l [
  {
   shapeType=Capsule
   length=17.6877
   radius=9.9398
   shapeName=thigh_r_capsule
   center=<-5.7473 -2.9372 3.8038>
   rotation=<rot -85.5988 -0.0007 176.0506>
  }
  {
   shapeType=Capsule
   length=15.5287
   radius=10.4581
   shapeName=thigh_l_capsule
   center=<-22.971 -1.8085 2.7459>
   rotation=<rot -88.8379 -0.0006 0.0003>
  }
 ]
 thigh_r [
  {
   shapeType=Capsule
   length=17.1634
   radius=10.4642
   shapeName=thigh_r_capsule
   center=<5.7016 2.2733 -3.8073>
   rotation=<rot -85.5988 -0.0007 -3.9494>
  }
  {
   shapeType=Capsule
   length=14.366
   radius=11.6208
   shapeName=thigh_l_capsule
   center=<22.9463 1.8085 -1.5298>
   rotation=<rot -88.8379 -0.0006 -179.9997>
  }
 ]
 thigh_twist_01_l [{
   shapeType=Capsule
   length=10.2611
   radius=6.1151
   shapeName=thigh_twist_01_r_capsule
   center=<4.0364 -1.4716 -6.7908>
   rotation=<rot 85 179.9989 -0.0006>
  }]
 thigh_twist_01_r [{
   shapeType=Capsule
   length=10.2611
   radius=6.1151
   shapeName=thigh_twist_01_r_capsule
   center=<-4.0364 1.4716 6.7908>
   rotation=<rot 85 179.9989 179.9994>
  }]
}
constraints [
 {
  bone1Name=dyn_jacket_side_r
  bone2Name=mstr_jacket_side_r
  swing1LimitDegrees=25
  swing2LimitDegrees=25
  twistLimitDegrees=5
  twistMotion=Locked
 }
 {
  bone1Name=dyn_jacket_side_l
  bone2Name=mstr_jacket_side_l
  swing1LimitDegrees=25
  swing2LimitDegrees=25
  twistLimitDegrees=5
  twistMotion=Locked
 }
 {
  bone1Name=thigh_twist_01_r
  bone2Name=thigh_r
  swing1LimitDegrees=45
  swing1Motion=Free
  swing2LimitDegrees=45
  swing2Motion=Free
  twistLimitDegrees=45
  twistMotion=Free
 }
 {
  bone1Name=thigh_twist_01_l
  bone2Name=thigh_l
  swing1LimitDegrees=45
  swing1Motion=Free
  swing2LimitDegrees=45
  swing2Motion=Free
  twistLimitDegrees=45
  twistMotion=Free
 }
 {bone1Name=thigh_r, bone2Name=pelvis, swing1LimitDegrees=25, swing2LimitDegrees=15, twistLimitDegrees=5, twistMotion=Locked}
 {bone1Name=thigh_l, bone2Name=pelvis, swing1LimitDegrees=5, swing2LimitDegrees=15, twistLimitDegrees=5, twistMotion=Locked}
 {
  bone1Name=dyn_jacket_front_r
  bone2Name=mstr_jacket_front_r
  swing1LimitDegrees=5
  swing1Motion=Locked
  swing2LimitDegrees=5
  swing2Motion=Locked
  twistLimitDegrees=25
 }
 {bone1Name=dyn_jacket_front_l, bone2Name=mstr_jacket_front_l, swing1LimitDegrees=5, swing2LimitDegrees=5, twistLimitDegrees=45}
]
