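# Physics asset for the back jacket flaps on the large male WarmShadeWeasel player body.
# Left, center and right dynamic jacket capsules are each constrained to a small sphere on their master jacket bone, with twist locked.
# Capsules on the pelvis, spine_02 and both thighs, two per thigh, approximate the body around the jacket.
# The thighs also constrain to the pelvis with locked twist, while spine_02 has no constraint.
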PhysicsAsset /BRCosmetics/Characters/Player/Male/Large/Bodies/M_LRG_WarmShadeWeasel/Meshes/M_LRG_WarmShadeWeasel_Physics_JacketBack
bodies {
 dyn_jacket_back_c [{
   shapeType=Capsule
   length=7.56
   radius=6.2035
   shapeName=dyn_jacketA_front_r_capsule
   center=<8.6866 5.2529 0>
   rotation=<rot -90 -90 90.0005>
  }]
 dyn_jacket_back_l [{
   shapeType=Capsule
   length=8.0612
   radius=5.7023
   shapeName=dyn_jacketA_front_r_capsule
   center=<-8.6866 -2.2784 0>
   rotation=<rot -90 -179.156 -0.8435>
  }]
 dyn_jacket_back_r [{
   shapeType=Capsule
   length=8.0612
   radius=5.7023
   shapeName=dyn_jacketA_front_r_capsule
   center=<8.6866 2.2784 0>
   rotation=<rot -90 -90 90.0005>
  }]
 mstr_jacket_back_c [{shapeType=Sphere, radius=2.0315, shapeName=mstr_jacket_side_r_sphere}]
 mstr_jacket_back_l [{shapeType=Sphere, radius=2.0315, shapeName=mstr_jacket_side_r_sphere}]
 mstr_jacket_back_r [{shapeType=Sphere, radius=2.0315, shapeName=mstr_jacket_side_r_sphere}]
 pelvis [{shapeType=Capsule, length=8.8103, radius=13.5027, shapeName=pelvis_capsule, center=<0 -2.3127 0>, rotation=<rot 0 -0.0002 -4.9122>}]
 spine_02 [{shapeType=Capsule, length=10.0796, radius=13.0749, shapeName=spine_02_capsule, center=<0 -1.7401 0>, rotation=<rot 0 179.999 -180>}]
 thigh_l [
  {
   shapeType=Capsule
   length=17.6877
   radius=9.9398
   shapeName=thigh_r_capsule
   center=<-5.7473 -2.9372 3.8038>
   rotation=<rot -85.5988 -0.0007 176.0506>
  }
  {
   shapeType=Capsule
   length=15.5287
   radius=10.4581
   shapeName=thigh_l_capsule
   center=<-22.971 -1.8085 2.7459>
   rotation=<rot -88.8379 -0.0006 0.0003>
  }
 ]
 thigh_r [
  {
   shapeType=Capsule
   length=17.1634
   radius=10.4642
   shapeName=thigh_r_capsule
   center=<5.7016 2.2733 -3.8073>
   rotation=<rot -85.5988 -0.0007 -3.9494>
  }
  {
   shapeType=Capsule
   length=14.366
   radius=11.6208
   shapeName=thigh_l_capsule
   center=<22.9463 1.8085 -1.5298>
   rotation=<rot -88.8379 -0.0006 -179.9997>
  }
 ]
}
constraints [
 {
  bone1Name=dyn_jacket_back_r
  bone2Name=mstr_jacket_back_r
  swing1LimitDegrees=60
  swing2LimitDegrees=15
  twistLimitDegrees=5
  twistMotion=Locked
 }
 {
  bone1Name=dyn_jacket_back_c
  bone2Name=mstr_jacket_back_c
  swing1LimitDegrees=60
  swing2LimitDegrees=15
  twistLimitDegrees=5
  twistMotion=Locked
 }
 {
  bone1Name=dyn_jacket_back_l
  bone2Name=mstr_jacket_back_l
  swing1LimitDegrees=60
  swing2LimitDegrees=15
  twistLimitDegrees=5
  twistMotion=Locked
 }
 {bone1Name=thigh_r, bone2Name=pelvis, swing1LimitDegrees=25, swing2LimitDegrees=15, twistLimitDegrees=5, twistMotion=Locked}
 {bone1Name=thigh_l, bone2Name=pelvis, swing1LimitDegrees=5, swing2LimitDegrees=15, twistLimitDegrees=5, twistMotion=Locked}
]
