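# Physics asset for the shorts of the large male WarmShadeWeasel body, named as a left-crouch variant, defining bodies for cloth-like secondary motion.
# Capsule dyn bodies hang from sphere mstr anchors through limited-swing, mostly twist-locked constraints, with child element capsules on the outer sides.
# Pocket elements are asymmetric: the right has a four-link chain, the left a single capsule.
# Unconstrained thigh capsules and a right-pocket box add collision volumes.
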PhysicsAsset /BRCosmetics/Characters/Player/Male/Large/Bodies/M_LRG_WarmShadeWeasel/Meshes/M_LRG_WarmShadeWeasel_Physics_shorts_Crouch_left
bodies {
 dyn_shortsElement_B1_r [{
   shapeType=Capsule
   length=0.6274
   radius=1.274
   shapeName=dyn_shorts_out_r_capsule
   center=<0.3725 0 0.0326>
   rotation=<rot -85 -0.0009 0.0001>
  }]
 dyn_shortsElement_B1a_r [{
   shapeType=Capsule
   length=0.9577
   radius=0.9315
   shapeName=dyn_shorts_out_r_capsule
   center=<1.0002 0 0.0875>
   rotation=<rot -85 -0.0009 0.0001>
  }]
 dyn_shortsElement_B2_r [{
   shapeType=Capsule
   length=0.5766
   radius=3.0938
   shapeName=dyn_shorts_out_r_capsule
   center=<1.7216 0 0.1506>
   rotation=<rot -4.9999 179.9996 180>
  }]
 dyn_shortsElement_B_l [{
   shapeType=Capsule
   length=2.4793
   radius=1.5734
   shapeName=dyn_shorts_out_r_capsule
   center=<-1.9353 0 -0.1693>
   rotation=<rot -85 -0.0009 -179.9999>
  }]
 dyn_shortsElement_B_r [{
   shapeType=Capsule
   length=0.6274
   radius=1.274
   shapeName=dyn_shorts_out_r_capsule
   center=<0.3725 0.257 0.0326>
   rotation=<rot -85 -0.0009 0.0001>
  }]
 dyn_shortsElement_out_l [{
   shapeType=Capsule
   length=5.9163
   radius=2.8255
   shapeName=dyn_shorts_out_r_capsule
   center=<-2.863 0 0.0167>
   rotation=<rot -89.4685 -0.0041 -179.9972>
  }]
 dyn_shortsElement_out_r [{
   shapeType=Capsule
   length=5.9163
   radius=2.8255
   shapeName=dyn_shorts_out_r_capsule
   center=<2.863 0 -0.0167>
   rotation=<rot -89.4685 -0.0041 0.0028>
  }]
 dyn_shorts_back_l [{
   shapeType=Capsule
   length=7.1545
   radius=5.3624
   shapeName=dyn_shorts_out_r_capsule
   center=<-6.7179 0.0001 -1.0033>
   rotation=<rot -86.9887 -0.0012 -179.9996>
  }]
 dyn_shorts_back_r [{
   shapeType=Capsule
   length=7.1545
   radius=5.3624
   shapeName=dyn_shorts_out_r_capsule
   center=<6.7179 -0.0001 1.0033>
   rotation=<rot -86.9887 -0.0012 0.0004>
  }]
 dyn_shorts_front_l [{
   shapeType=Capsule
   length=7.9251
   radius=4.5918
   shapeName=dyn_shorts_out_r_capsule
   center=<-6.7179 0.0001 -1.0033>
   rotation=<rot -86.9887 -0.0012 -179.9996>
  }]
 dyn_shorts_front_r [{
   shapeType=Capsule
   length=7.9251
   radius=4.5918
   shapeName=dyn_shorts_out_r_capsule
   center=<6.7179 -0.0001 1.0033>
   rotation=<rot -86.9887 -0.0012 0.0004>
  }]
 dyn_shorts_inner_l [{
   shapeType=Capsule
   length=7.1545
   radius=5.3624
   shapeName=dyn_shorts_out_r_capsule
   center=<-6.7179 0.0001 -1.0033>
   rotation=<rot -86.9887 -0.0012 -179.9996>
  }]
 dyn_shorts_inner_r [{
   shapeType=Capsule
   length=7.1545
   radius=5.3624
   shapeName=dyn_shorts_out_r_capsule
   center=<6.7179 -0.0001 1.0033>
   rotation=<rot -86.9887 -0.0012 0.0004>
  }]
 dyn_shorts_out_l [{
   shapeType=Capsule
   length=5.9163
   radius=2.8255
   shapeName=dyn_shorts_out_r_capsule
   center=<-2.8398 0 -0.2485>
   rotation=<rot -85 -0.0009 -179.9999>
  }]
 dyn_shorts_out_r [{
   shapeType=Capsule
   length=5.9163
   radius=2.8255
   shapeName=dyn_shorts_out_r_capsule
   center=<2.8398 0 0.2485>
   rotation=<rot -85 -0.0009 0.0001>
  }]
 mstr_shorts_back_l [{shapeType=Sphere, radius=2.1962, shapeName=mstr_shorts_pocket_r_sphere}]
 mstr_shorts_back_r [{shapeType=Sphere, radius=2.1962, shapeName=mstr_shorts_pocket_r_sphere}]
 mstr_shorts_front_l [{shapeType=Sphere, radius=2.1962, shapeName=mstr_shorts_pocket_r_sphere}]
 mstr_shorts_front_r [{shapeType=Sphere, radius=2.1962, shapeName=mstr_shorts_pocket_r_sphere}]
 mstr_shorts_inner_l [{shapeType=Sphere, radius=2.1962, shapeName=mstr_shorts_pocket_r_sphere}]
 mstr_shorts_inner_r [{shapeType=Sphere, radius=2.1962, shapeName=mstr_shorts_pocket_r_sphere}]
 mstr_shorts_out_l [{shapeType=Sphere, radius=2.1962, shapeName=mstr_shorts_pocket_r_sphere}]
 mstr_shorts_out_r [{shapeType=Sphere, radius=2.1962, shapeName=mstr_shorts_pocket_r_sphere}]
 mstr_shorts_pocket_l [{shapeType=Sphere, radius=2.1962, shapeName=mstr_shorts_pocket_r_sphere}]
 mstr_shorts_pocket_r [
  {shapeType=Sphere, radius=2.1962, shapeName=mstr_shorts_pocket_r_sphere}
  {shapeType=Box, extentX=13.9757, extentY=7.5, extentZ=7.5, shapeName=mstr_shorts_pocket_r_box, center=<8.9897 0 0>}
 ]
 thigh_l [{shapeType=Capsule, length=21.7, radius=8.3, shapeName=thigh_r_capsule, center=<-22.8599 -1.5654 0>, rotation=<rot 90 178.3124 -1.6875>}]
 thigh_r [{shapeType=Capsule, length=21.7, radius=8.3, shapeName=thigh_r_capsule, center=<22.8599 1.5654 0>, rotation=<rot 90 6.7214 6.7216>}]
}
constraints [
 {
  bone1Name=dyn_shortsElement_out_r
  bone2Name=dyn_shorts_out_r
  swing1LimitDegrees=37.36
  swing1Motion=Locked
  swing2LimitDegrees=50.32
  twistLimitDegrees=5
  twistMotion=Locked
 }
 {
  bone1Name=dyn_shortsElement_B_r
  bone2Name=mstr_shorts_pocket_r
  swing1LimitDegrees=20
  swing2LimitDegrees=20
  swing2Motion=Locked
  twistLimitDegrees=45
  twistMotion=Locked
 }
 {
  bone1Name=dyn_shorts_out_r
  bone2Name=mstr_shorts_out_r
  swing1LimitDegrees=10
  swing2LimitDegrees=10
  twistLimitDegrees=5
  twistMotion=Locked
 }
 {
  bone1Name=dyn_shortsElement_B_l
  bone2Name=mstr_shorts_pocket_l
  swing1LimitDegrees=20
  swing2LimitDegrees=20
  twistLimitDegrees=45
  twistMotion=Locked
 }
 {
  bone1Name=dyn_shorts_front_r
  bone2Name=mstr_shorts_front_r
  swing1LimitDegrees=5
  swing2LimitDegrees=10
  twistLimitDegrees=45
  twistMotion=Locked
 }
 {
  bone1Name=dyn_shorts_back_r
  bone2Name=mstr_shorts_back_r
  swing1LimitDegrees=5
  swing2LimitDegrees=10
  twistLimitDegrees=45
  twistMotion=Locked
 }
 {
  bone1Name=dyn_shorts_front_l
  bone2Name=mstr_shorts_front_l
  swing1LimitDegrees=5
  swing2LimitDegrees=10
  twistLimitDegrees=45
  twistMotion=Locked
 }
 {
  bone1Name=dyn_shorts_inner_l
  bone2Name=mstr_shorts_inner_l
  swing1LimitDegrees=5
  swing2LimitDegrees=10
  twistLimitDegrees=45
  twistMotion=Locked
 }
 {
  bone1Name=dyn_shorts_inner_r
  bone2Name=mstr_shorts_inner_r
  swing1LimitDegrees=5
  swing2LimitDegrees=10
  twistLimitDegrees=45
  twistMotion=Locked
 }
 {
  bone1Name=dyn_shorts_back_l
  bone2Name=mstr_shorts_back_l
  swing1LimitDegrees=5
  swing2LimitDegrees=10
  twistLimitDegrees=45
  twistMotion=Locked
 }
 {
  bone1Name=dyn_shorts_out_l
  bone2Name=mstr_shorts_out_l
  swing1LimitDegrees=10
  swing2LimitDegrees=10
  twistLimitDegrees=5
  twistMotion=Locked
 }
 {
  bone1Name=dyn_shortsElement_out_l
  bone2Name=dyn_shorts_out_l
  swing1LimitDegrees=37.36
  swing1Motion=Locked
  swing2LimitDegrees=50.32
  twistLimitDegrees=5
  twistMotion=Locked
 }
 {bone1Name=dyn_shortsElement_B2_r, bone2Name=dyn_shortsElement_B1a_r, swing1LimitDegrees=25, swing2LimitDegrees=25, twistLimitDegrees=5}
 {bone1Name=dyn_shortsElement_B1a_r, bone2Name=dyn_shortsElement_B1_r, swing1LimitDegrees=45, swing2LimitDegrees=25, twistLimitDegrees=5}
 {bone1Name=dyn_shortsElement_B1_r, bone2Name=dyn_shortsElement_B_r, swing1LimitDegrees=20, swing2LimitDegrees=10, twistLimitDegrees=5}
]
