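# Post-copy cosmetic AnimBP for the large male WarmShadeWeasel body, adding secondary motion to its shorts and jacket.
# The parent mesh pose runs through a Control Rig fed a smoothed jump/fall alpha, then chained RigidBody sims for shorts and back jacket.
# Calf yaw crossfades crouch and standard shorts physics per leg, while spine yaw weights the back jacket.
# A final jacket sim on the cached pose is bypassed while crouching or emoting.
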AnimBlueprint /BRCosmetics/Characters/Player/Male/Large/Bodies/M_LRG_WarmShadeWeasel/Meshes/M_LRG_WarmShadeWeasel_AnimBP
target_skeleton=&/M_LRG_WarmShadeWeasel_Skeleton
graphs [
 AnimGraph {nodes [
   n1 = RigidBody!(ComponentPose=Local To Component(LocalPose=Use cached pose 'CachA'()), Alpha, OverrideWorldGravity:Vector="0,0,0", ExternalForce) -> (Pose) {
    bOverrideWorldGravity=true
    overridePhysicsAsset=&/M_LRG_WarmShadeWeasel_Physics_Jacket1
    simulationSpace=BaseBoneSpace
    alphaBoolBlend {blendOption=Linear}
    baseBoneRef {boneName=pelvis}
    componentAppliedLinearAccClamp=<1000 1000 1000>
    componentLinearAccScale=<0.2 0.2 0.2>
    componentLinearVelScale=<0.2 0.2 0.2>
    simSpaceSettings {maxAngularAcceleration=10, maxAngularVelocity=10, maxLinearAcceleration=150, maxLinearVelocity=50, velocityScaleZ=0, worldAlpha=0.2}
   }
   Output Pose(Result=Blend Poses by bool(BlendPose_0=Component To Local(ComponentPose=@n1), BlendPose_1=Use cached pose 'CachA'(), bActiveValue=Get JacketDyn() -> (JacketDyn:bool)) {blendType=Linear}) {layerGroup=None}
   n2 = Control Rig!(
    Source=Copy Pose From Mesh() {bUseAttachedParent=true}
    Alpha
    falling alpha:double=Get falling alpha() -> (falling alpha:double)
    RBAN:bool=Get bRigidBodyEnabled()
   ) -> (Pose) {
    alphaBoolBlend {blendOption=Linear}
    controlRigAssetReference {blueprintRigClass=&/M_LRG_WarmShadeWeasel_CtrlRig.M_LRG_WarmShadeWeasel_CtrlRig_C}
   }
   n3 = RigidBody(
    ComponentPose=Local To Component(LocalPose=@n2)
    Alpha=Get Shorts r crouch alpha() -> (Shorts r crouch alpha:double)
    OverrideWorldGravity="0,0,0"
   ) {
    bOverrideWorldGravity=true
    overridePhysicsAsset=&/M_LRG_WarmShadeWeasel_Physics_shorts_Crouch
    simulationSpace=BaseBoneSpace
    alphaBoolBlend {blendOption=Linear}
    baseBoneRef {boneName=thigh_twist_01_r}
    componentAppliedLinearAccClamp=<3000 3000 3000>
    componentLinearAccScale=<1 1 1>
    componentLinearVelScale=<1 1 1>
    simSpaceSettings {maxAngularAcceleration=150, maxAngularVelocity=150, maxLinearAcceleration=150, maxLinearVelocity=150, velocityScaleZ=0, worldAlpha=1}
   }
   n4 = RigidBody(ComponentPose=@n3, Alpha=Get Shorts l crouch alpha() -> (Shorts l crouch alpha:double), OverrideWorldGravity="0,0,0") {
    bOverrideWorldGravity=true
    overridePhysicsAsset=&/M_LRG_WarmShadeWeasel_Physics_shorts_Crouch_left
    simulationSpace=BaseBoneSpace
    alphaBoolBlend {blendOption=Linear}
    baseBoneRef {boneName=thigh_twist_01_l}
    componentAppliedLinearAccClamp=<3000 3000 3000>
    componentLinearAccScale=<1 1 1>
    componentLinearVelScale=<1 1 1>
    simSpaceSettings {maxAngularAcceleration=150, maxAngularVelocity=150, maxLinearAcceleration=150, maxLinearVelocity=150, velocityScaleZ=0, worldAlpha=0.6}
   }
   n5 = RigidBody(ComponentPose=@n4, Alpha="Get Shorts r  alpha"() -> ("Shorts r  alpha":double), OverrideWorldGravity="0,0,0") {
    bOverrideWorldGravity=true
    overridePhysicsAsset=&/M_LRG_WarmShadeWeasel_Physics_shorts_R
    simulationSpace=BaseBoneSpace
    alphaBoolBlend {blendOption=Linear}
    baseBoneRef {boneName=thigh_twist_01_r}
    componentAppliedLinearAccClamp=<1200 1200 1200>
    componentLinearAccScale=<1 1 1>
    componentLinearVelScale=<1 1 1>
    simSpaceSettings {maxAngularAcceleration=150, maxAngularVelocity=150, maxLinearAcceleration=150, maxLinearVelocity=150, velocityScaleZ=0, worldAlpha=1}
   }
   n6 = RigidBody(ComponentPose=@n5, Alpha="Get Shorts l  alpha"() -> ("Shorts l  alpha":double), OverrideWorldGravity="0,0,0") {
    bOverrideWorldGravity=true
    overridePhysicsAsset=&/M_LRG_WarmShadeWeasel_Physics_shorts_left
    simulationSpace=BaseBoneSpace
    alphaBoolBlend {blendOption=Linear}
    baseBoneRef {boneName=thigh_twist_01_l}
    componentAppliedLinearAccClamp=<1200 1200 1200>
    componentLinearAccScale=<1 1 1>
    componentLinearVelScale=<1 1 1>
    simSpaceSettings {maxAngularAcceleration=150, maxAngularVelocity=150, maxLinearAcceleration=150, maxLinearVelocity=150, velocityScaleZ=0, worldAlpha=1}
   }
   n7 = RigidBody(ComponentPose=@n6, Alpha=Get BackJacketAlpha() -> (BackJacketAlpha:double), OverrideWorldGravity="0,0,0") {
    bOverrideWorldGravity=true
    overridePhysicsAsset=&/M_LRG_WarmShadeWeasel_Physics_JacketBack
    simulationSpace=BaseBoneSpace
    alphaBoolBlend {blendOption=Linear}
    baseBoneRef {boneName=pelvis}
    componentAppliedLinearAccClamp=<500 500 500>
    componentLinearAccScale=<1 1 1>
    componentLinearVelScale=<1 1 1>
    simSpaceSettings {maxAngularAcceleration=10, maxAngularVelocity=10, maxLinearAcceleration=10, maxLinearVelocity=5, velocityScaleZ=0, worldAlpha=0.2}
   }
   Save cached pose 'CachA'(Pose=Component To Local(ComponentPose=@n7)) {cacheName=CachA}
  ]}
 EventGraph {nodes [
   n1 = Event BlueprintUpdateAnimation()
   n2 = Self-Reference()
   > Is Valid(InputObject=@n2)
   > n3 = Gravity Override State Function DEPRECATED(CharacterPartAnimInstance=@n2, GameTimeInSeconds=@n1.DeltaTimeX)
   > Set Gravity Override(Gravity Override=@n3)
   n4 = GetOwnerSkeletalMeshComponent(self=@n2)
   > n5 = Set Shorts r crouch alpha(Shorts r crouch alpha:double=Remap Float(
     Input=GetSocketTransform!(self=@n4, InSocketName=calf_r, TransformSpace=RTS_ParentBoneSpace) -> (
      ReturnValue_Location:Vector
      ReturnValue_Rotation_Roll:float
      ReturnValue_Rotation_Pitch:float
      ReturnValue_Rotation_Yaw:float
      ReturnValue_Scale:Vector
     ).ReturnValue_Rotation_Yaw
     OldRange=(X=70,Y=120)
     NewRange=(X=0,Y=1)
     Clamp=true
    ) -> (Output=~)) -> (Output_Get:double)
   > "Set Shorts r  alpha"("Shorts r  alpha":double=float - float(A=1, B=@n5)) -> (Output_Get:double)
   > n6 = Set Shorts l crouch alpha(Shorts l crouch alpha:double=Remap Float(
     Input=GetSocketTransform(self=@n4, InSocketName=calf_l, TransformSpace=RTS_ParentBoneSpace).ReturnValue_Rotation_Yaw
     OldRange=(X=70,Y=120)
     NewRange=(X=0,Y=1)
     Clamp=true
    ) -> (Output=~)) -> (Output_Get:double)
   > "Set Shorts l  alpha"("Shorts l  alpha":double=float - float(A=1, B=@n6)) -> (Output_Get:double)
   n7 = OR Boolean(A=Get bIsJumping(), B=Get bIsFalling())
   > Set falling alpha(falling alpha:double=FInterpTo(
     Current=Get falling alpha()
     Target="To Float (Boolean)"(InBool=@n7)
     DeltaTime=GetWorldDeltaSeconds!() -> (ReturnValue:double)
     InterpSpeed=25
    )) -> (Output_Get:double)
   n8 = Get bIsPlayingEmote()
   n9 = Get bIsCrouching()
   > Set JacketDyn(JacketDyn:bool=NOT Boolean(A=OR Boolean(A=@n8, B=@n9))) -> (Output_Get:bool)
   n10 = GetOwnerSkeletalMeshComponent()
   > Set BackJacketAlpha(BackJacketAlpha:double=Remap Float(
     Input=GetSocketTransform(self=@n10, InSocketName=spine_01, TransformSpace=RTS_ParentBoneSpace).ReturnValue_Rotation_Yaw
     OldRange=(X=-6,Y=-45)
     NewRange=(X=1,Y=0)
     Clamp=true
    ) -> (Output=~)) -> (Output_Get:double)
   NOT Boolean(A=Get bIsOnGround() -> (bIsOnGround:bool))
  ]}
]
variables [
 BackJacketAlpha {type=number}
 falling alpha {type=number, default=1}
 Gravity Override {type=Vector}
 JacketDyn {type=boolean}
 "Shorts l  alpha" {type=number}
 Shorts l crouch alpha {type=number}
 "Shorts r  alpha" {type=number}
 Shorts r crouch alpha {type=number}
]
