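# Control Rig for the face accessory of the WarmShadeWeasel large male player body, applying corrective offsets to hood bones as the head moves.
# A sequence first uses a head-versus-spine_04 spherical pose reader to weight back and nape offsets.
# It then remaps head Euler X, relative to head_null and spine_05, into mostly unclamped weights for front hood and clavicle offsets.
# A back-hood parent-constraint chain and a zero-offset node are disconnected from execution.
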ControlRigBlueprint /BRCosmetics/Characters/Player/Male/Large/Bodies/M_LRG_WarmShadeWeasel/Meshes/Parts/M_LRG_WarmShadeWeasel_FaceAcc_CtrlRig
graphs [
 RigVMFunctionLibrary {links [], nodes []}
 RigVMModel {nodes [
   RigUnit_GetRelativeTransformForItem = &RigUnit_GetRelativeTransformForItem(Child=(Type=Bone,Name="head"), Parent=(Type=Null,Name="head_null"), bChildInitial=False, bParentInitial=False)
   RigVMFunction_MathQuaternionToEuler = &RigVMFunction_MathQuaternionToEuler(RotationOrder=ZYX, Value=@RigUnit_GetRelativeTransformForItem.RelativeTransform.Rotation)
   n1 = &RigVMFunction_MathDoubleRemap(SourceMaximum=-35, SourceMinimum=0, TargetMaximum=1, TargetMinimum=0, Value=@RigVMFunction_MathQuaternionToEuler.Result.Z)
   Subtract = &RigVMFunction_MathDoubleSub(A=1, B=@n1.Result)
   &RigUnit_ParentConstraint(
    AdvancedSettings=(InterpolationType=Average,RotationOrderForFilter=XZY)
    Child=(Type=Bone,Name="mstr_HoodUp_back_c")
    ExecutePin=()
    Parents=((Item=(Type=Bone,Name="spine_05"),Weight=1),(Item=(Type=Bone,Name="head"),Weight=1))
    Weight=1
    Parents.0.Weight=@n1.Result
    Parents.1.Weight=@Subtract.Result
   )
   Remap = &RigVMFunction_MathDoubleRemap(SourceMaximum=-20, SourceMinimum=0, TargetMaximum=1, TargetMinimum=0, Value=@RigVMFunction_MathQuaternionToEuler.Result.Z)
   > &RigUnit_ModifyTransforms(
    ItemToModify=((Item=(Type=Bone,Name="mstr_HoodUp_back_c"),Transform=(Rotation=(X=0,Y=0,Z=0.1731,W=0.9849),Translation=(X=-8.744,Y=-7.8004,Z=-6.7341),Scale3D=(X=1,Y=1,Z=1))))
    WeightMinimum=0
    Weight=@Remap.Result
   )
   RigUnit_BeginExecution = &RigUnit_BeginExecution()
   RigVMFunction_Sequence = &RigVMFunction_Sequence(ExecuteContext=@RigUnit_BeginExecution.ExecutePin)
   DISPATCH_RigVMDispatch_Print = RigVMDispatchNode(
    Enabled=false
    ScreenColor=(R=1,G=1,B=1,A=1)
    ScreenDuration=0.05
    Value=@RigVMFunction_MathQuaternionToEuler.Result
    ExecuteContext=@RigVMFunction_Sequence.A
   ) {template="DISPATCH_RigVMDispatch_Print(in Prefix,in Value,in Enabled,in ScreenDuration,in ScreenColor)"}
   > n2 = &RigUnit_SphericalPoseReader(
    ActiveRegionScaleFactors=(PositiveWidth=1,NegativeWidth=1,PositiveHeight=1,NegativeHeight=1)
    ActiveRegionSize=0.1
    Debug=(bDrawDebug=true,bDraw2D=False,bDrawLocalAxes=False,DebugScale=25,DebugSegments=20,DebugThickness=0.25)
    DriverAxis=(X=1,Y=0,Z=0)
    DriverItem=(Type=Bone,Name="head")
    FalloffRegionScaleFactors=(PositiveWidth=1,NegativeWidth=1,PositiveHeight=1,NegativeHeight=1)
    FalloffSize=0.4
    FlipHeightScaling=False
    FlipWidthScaling=False
    OptionalParentItem=(Type=Bone,Name="spine_04")
    RotationOffset=(X=-70,Y=0,Z=15)
   )
   > &RigUnit_ModifyTransforms(
    ItemToModify="((Item=(Type=Bone,Name=\"mstr_HoodUp_back_c\"),Transform=(Rotation=(X=0.0471,Y=-0.1897,Z=-0.0091,W=-0.9807),Translation=(X=-7.0124,Y=-3.3889,Z=0.6774),Scale3D=(X=1,Y=1,Z=1))),(Item=(Type=Bone,Name=\"mstr_HoodDown_nape_c\"),Transform=(Rotation=(X=0,Y=0,Z=0,W=1) ...(truncated)"
    WeightMinimum=0
    Weight=@n2.OutputParam
   )
   &RigUnit_ModifyTransforms(
    ItemToModify="((Item=(Type=Bone,Name=\"mstr_HoodUp_Front_r\"),Transform=(Rotation=(X=-0.0468,Y=0.2113,Z=0.2113,W=-0.9532),Translation=(X=-9.1641,Y=-1.1395,Z=-10.4292),Scale3D=(X=1,Y=1,Z=1))),(Item=(Type=Bone,Name=\"mstr_HoodUp_Front_l\"),Transform=(Rotation=(X=-0.342,Y=0,Z= ...(truncated)"
    WeightMinimum=0
    Weight=&RigVMFunction_MathDoubleRemap(SourceMaximum=-45, SourceMinimum=0, TargetMaximum=1, TargetMinimum=0, bClamp=False, Value=@RigVMFunction_MathQuaternionToEuler.Result.X).Result
    ExecutePin=@RigVMFunction_Sequence.B
   )
   > &RigUnit_ModifyTransforms(
    ItemToModify="((Item=(Type=Bone,Name=\"mstr_HoodUp_Front_r\"),Transform=(Rotation=(X=-0.1305,Y=0,Z=0,W=-0.9914),Translation=(X=-6.8698,Y=-2.0748,Z=3.4349),Scale3D=(X=1,Y=1,Z=1))),(Item=(Type=Bone,Name=\"mstr_HoodUp_Front_l\"),Transform=(Rotation=(X=0.0057,Y=-0.0432,Z=0.1304 ...(truncated)"
    WeightMinimum=0
    Weight=&RigVMFunction_MathDoubleRemap(SourceMaximum=45, SourceMinimum=0, TargetMaximum=1, TargetMinimum=0, bClamp=False, Value=@RigVMFunction_MathQuaternionToEuler.Result.X).Result
   )
   n3 = &RigVMFunction_MathDoubleRemap(
    SourceMaximum=69
    SourceMinimum=0
    TargetMaximum=1
    TargetMinimum=0
    bClamp=False
    Value=&RigVMFunction_MathQuaternionToEuler(
     RotationOrder=ZYX
     Value=&RigUnit_GetRelativeTransformForItem(Child=(Type=Bone,Name="head"), Parent=(Type=Bone,Name="spine_05"), bChildInitial=False, bParentInitial=False).RelativeTransform.Rotation
    ).Result.X
   )
   > RigVMDispatchNode(Enabled=false, ScreenColor=(R=1,G=1,B=1,A=1), ScreenDuration=0.05, Value=@n3.Result) {template="DISPATCH_RigVMDispatch_Print(in Prefix,in Value,in Enabled,in ScreenDuration,in ScreenColor)"}
   > &RigUnit_ModifyTransforms(
    ItemToModify="((Item=(Type=Bone,Name=\"mstr_HoodUp_Front_r\"),Transform=(Rotation=(X=-0.1441,Y=0.1054,Z=0.4782,W=-0.8599),Translation=(X=1.6002,Y=-1.8178,Z=4.1936),Scale3D=(X=0.8761,Y=1.7229,Z=0.794))),(Item=(Type=Bone,Name=\"clavicle_r\"),Transform=(Rotation=(X=0,Y=0,Z=0,W ...(truncated)"
    WeightMinimum=0
    Weight=@n3.Result
   )
   &RigUnit_ModifyTransforms(
    ExecutePin=()
    ItemToModify="((Item=(Type=Bone,Name=\"mstr_HoodUp_back_c\"),Transform=(Rotation=(X=0,Y=0,Z=0,W=1),Translation=(X=0,Y=0,Z=0),Scale3D=(X=1,Y=1,Z=1))),(Item=(Type=Bone,Name=\"mstr_HoodDown_nape_c\"),Transform=(Rotation=(X=0,Y=0,Z=0,W=1),Translation=(X=0,Y=0,Z=0),Scale3D=(X=1, ...(truncated)"
    Weight=1
    WeightMinimum=0
   )
  ]}
]
hierarchy [
 root {loc=<0 0 0>, rot=<0 0 0 1>}
 pelvis {parent=root, loc=<0 -0.7321 92.4218>, rot=<0.0127 0.707 -0.0127 -0.707>}
 spine_01 {parent=pelvis, loc=<5.6002 -0.5128 0>, rot=<0 0 0.034 -0.9994>}
 spine_02 {parent=spine_01, loc=<8.0753 0.0139 0>, rot=<0 0 0.0238 -0.9997>}
 spine_03 {parent=spine_02, loc=<7.0393 0 0>, rot=<0 0 0.0149 -0.9999>}
 spine_04 {parent=spine_03, loc=<8.1024 0 0>, rot=<0 0 -0.0132 -0.9999>}
 spine_05 {parent=spine_04, loc=<10.0146 0 0>, rot=<0 0 -0.0015 -1>}
 neck_01 {parent=spine_05, loc=<10.815 1.9196 0>, rot=<0 0 -0.1864 -0.9825>}
 neck_02 {parent=neck_01, loc=<4.7347 0.3045 0>, rot=<0 0 0.0363 -0.9993>}
 head {parent=neck_02, loc=<5.7632 0.5552 0>, rot=<0 0 0.115 -0.9934>}
 faceAccRoot {parent=head, loc=<0 0 0>, rot=<0 0 0 1>}
 mstr_HoodUp_Front_r {parent=faceAccRoot, loc=<-2.768 3.9429 11.8083>, rot=<-0.0197 -0.7241 -0.6889 0.0267>}
 dyn_HoodUp_Front_r {parent=mstr_HoodUp_Front_r, loc=<0 0 0>, rot=<0 0 0 1>}
 mstr_HoodUp_Front_l {parent=faceAccRoot, loc=<-2.7679 3.9429 -11.8083>, rot=<-0.7241 0.0197 -0.0267 -0.6889>}
 dyn_HoodUp_Front_l {parent=mstr_HoodUp_Front_l, loc=<0 0 0>, rot=<0 0 0 1>}
 mstr_faceAcc_head {parent=faceAccRoot, loc=<0 0 0>, rot=<0 0 0 1>}
 FACIAL_C_FacialRoot {parent=head, loc=<3.8999 -1.8321 -0.0088>, rot=<0.4672 -0.5308 -0.4673 -0.5307>}
 FACIAL_L_CheekInner {parent=FACIAL_C_FacialRoot, loc=<1.9847 0.9895 7.4432>, rot=<-0.0835 -0.096 -0.1933 -0.9728>}
 FACIAL_R_CheekInner {parent=FACIAL_C_FacialRoot, loc=<-1.9922 0.9889 7.4472>, rot=<-0.0839 0.0959 0.1935 -0.9728>}
 FACIAL_L_CheekOuter {parent=FACIAL_C_FacialRoot, loc=<3.7436 1.3416 6.0391>, rot=<0.0987 -0.3718 0.1182 -0.9155>}
 FACIAL_R_CheekOuter {parent=FACIAL_C_FacialRoot, loc=<-3.7606 1.3425 6.037>, rot=<0.0984 0.3718 -0.118 -0.9155>}
 FACIAL_L_NasolabialBulge {parent=FACIAL_C_FacialRoot, loc=<3.5762 4.7127 5.3284>, rot=<0.0925 -0.3761 0.0379 -0.9212>}
 FACIAL_R_NasolabialBulge {parent=FACIAL_C_FacialRoot, loc=<-3.589 4.7069 5.3213>, rot=<0.0925 0.376 -0.0379 -0.9212>}
 FACIAL_L_NasolabialFurrow {parent=FACIAL_C_FacialRoot, loc=<2.837 3.6242 8.3571>, rot=<0.0082 -0.1421 -0.3576 -0.923>}
 FACIAL_R_NasolabialFurrow {parent=FACIAL_C_FacialRoot, loc=<-2.8711 3.6206 8.3541>, rot=<0.008 0.1419 0.3577 -0.923>}
 FACIAL_L_CheekLower {parent=FACIAL_C_FacialRoot, loc=<3.8349 4.5927 4.7417>, rot=<0.0684 -0.4243 0.0737 -0.8999>}
 FACIAL_R_CheekLower {parent=FACIAL_C_FacialRoot, loc=<-3.8445 4.5866 4.7344>, rot=<0.0684 0.4243 -0.0737 -0.8999>}
 FACIAL_C_Nose {parent=FACIAL_C_FacialRoot, loc=<-0.0045 0.5875 9.3789>, rot=<-0.0945 0 0.0001 -0.9955>}
 FACIAL_C_NoseLower {parent=FACIAL_C_Nose, loc=<-0.004 2.6278 -0.3899>, rot=<0.1306 0 0 -0.9914>}
 FACIAL_L_Nostril {parent=FACIAL_C_Nose, loc=<1.5048 1.9366 -0.4727>, rot=<0.0173 0.1305 0.0023 -0.9913>}
 FACIAL_R_Nostril {parent=FACIAL_C_Nose, loc=<-1.519 1.9334 -0.4612>, rot=<0.0173 -0.1305 -0.0023 -0.9913>}
 FACIAL_C_MouthUpper {parent=FACIAL_C_FacialRoot, loc=<-0.0062 5.0767 6.5656>, rot=<0.0375 0 0 -0.9993>}
 FACIAL_C_LipUpper {parent=FACIAL_C_MouthUpper, loc=<-0.0011 -0.5524 2.9505>, rot=<-0.0872 0 0 -0.9962>}
 FACIAL_L_LipUpper {parent=FACIAL_C_MouthUpper, loc=<1.4583 -0.2579 2.6091>, rot=<-0.0778 -0.16 -0.0378 -0.9833>}
 FACIAL_R_LipUpper {parent=FACIAL_C_MouthUpper, loc=<-1.5103 -0.2732 2.6095>, rot=<-0.0778 0.16 0.0378 -0.9833>}
 FACIAL_L_LipUpperOuter {parent=FACIAL_C_MouthUpper, loc=<2.1592 0.0588 1.8429>, rot=<-0.0984 -0.2292 -0.0982 -0.9634>}
 FACIAL_R_LipUpperOuter {parent=FACIAL_C_MouthUpper, loc=<-2.2257 0.0403 1.8258>, rot=<-0.0984 0.2292 0.0982 -0.9634>}
 FACIAL_L_LipCorner {parent=FACIAL_C_MouthUpper, loc=<2.5683 0.4927 1.128>, rot=<0 -0.2077 0 -0.9782>}
 FACIAL_R_LipCorner {parent=FACIAL_C_MouthUpper, loc=<-2.6329 0.4619 1.1029>, rot=<0 0.2077 0 -0.9782>}
 FACIAL_L_Masseter {parent=FACIAL_C_FacialRoot, loc=<6.4341 1.6111 3.2563>, rot=<0.12 0.1131 0.0021 -0.9863>}
 FACIAL_R_Masseter {parent=FACIAL_C_FacialRoot, loc=<-6.4305 1.6059 3.2581>, rot=<0.1192 -0.1131 -0.0022 -0.9864>}
 FACIAL_C_UnderChin {parent=FACIAL_C_FacialRoot, loc=<0.0032 10.3811 3.3555>, rot=<0.5247 0 0 -0.8513>}
 FACIAL_L_UnderChin {parent=FACIAL_C_FacialRoot, loc=<3.6843 8.4508 1.7042>, rot=<0.4325 -0.3517 0.0951 -0.8247>}
 FACIAL_R_UnderChin {parent=FACIAL_C_FacialRoot, loc=<-3.6784 8.4438 1.6931>, rot=<0.4326 0.3517 -0.0951 -0.8247>}
 FACIAL_C_LowerLipRotation {parent=FACIAL_C_FacialRoot, loc=<-0.0022 2.0528 1.373>, rot=<0.377 0 0 -0.9262>}
 FACIAL_C_MouthLower {parent=FACIAL_C_LowerLipRotation, loc=<-0.0045 -1.4625 5.8282>, rot=<-0.342 0 0 -0.9397>}
 FACIAL_C_LipLower {parent=FACIAL_C_MouthLower, loc=<0.0109 1.1729 2.571>, rot=<0.0872 0 0 -0.9962>}
 FACIAL_L_LipLower {parent=FACIAL_C_MouthLower, loc=<1.2652 1.2437 2.378>, rot=<0.0927 -0.1294 0.0272 -0.9869>}
 FACIAL_R_LipLower {parent=FACIAL_C_MouthLower, loc=<-1.2677 1.1863 2.3605>, rot=<0.0927 0.1294 -0.0272 -0.9869>}
 FACIAL_L_LipLowerOuter {parent=FACIAL_C_MouthLower, loc=<2.1158 0.8433 1.6554>, rot=<0.0913 -0.232 0.0318 -0.9679>}
 FACIAL_R_LipLowerOuter {parent=FACIAL_C_MouthLower, loc=<-2.1389 0.8229 1.6314>, rot=<0.0913 0.232 -0.0318 -0.9679>}
 FACIAL_C_Jaw {parent=FACIAL_C_FacialRoot, loc=<-0.0022 2.0528 1.373>, rot=<0.377 0 0 -0.9262>}
 FACIAL_C_Jawline {parent=FACIAL_C_Jaw, loc=<-0.002 3.163 8.6587>, rot=<0.2616 0 0 -0.9652>}
 FACIAL_L_ChinSide {parent=FACIAL_C_Jaw, loc=<3.4442 0.771 8.0581>, rot=<-0.2091 -0.3731 0.0548 -0.9023>}
 FACIAL_R_ChinSide {parent=FACIAL_C_Jaw, loc=<-3.4521 0.7835 8.0548>, rot=<-0.2095 0.3732 -0.0546 -0.9022>}
 FACIAL_C_Chin {parent=FACIAL_C_Jaw, loc=<-0.0029 1.3168 8.6298>, rot=<-0.2492 0 0 -0.9684>}
 FACIAL_L_Ear {parent=FACIAL_C_FacialRoot, loc=<6.5663 -0.0697 1.1745>, rot=<0.0139 -0.0017 -0.0261 -0.9996>}
 FACIAL_R_Ear {parent=FACIAL_C_FacialRoot, loc=<-6.583 -0.0715 1.1752>, rot=<0.0134 0.0017 0.0262 -0.9996>}
 FACIAL_L_Jawline {parent=FACIAL_C_FacialRoot, loc=<4.7465 6.2085 3.1333>, rot=<0.2557 0.1549 -0.0622 -0.9522>}
 FACIAL_R_Jawline {parent=FACIAL_C_FacialRoot, loc=<-4.75 6.2078 3.1248>, rot=<0.2552 -0.155 0.0622 -0.9523>}
 mstr_HoodUp_back_c {parent=head, loc=<2.9518 -12.948 0>, rot=<0.1036 -0.9946 0 0>}
 mstr_HoodDown_nape_c {parent=head, loc=<12.9459 -10.3724 0>, rot=<0.1036 -0.9946 0 0>}
 head_null {parent=neck_02, type=Null, loc=<5.7632 0.5552 0>, rot=<0 0 0.115 -0.9934>}
 mstr_HoodUp_back_r {parent=neck_01, loc=<5.8686 -13.1967 6.1725>, rot=<-0.0553 -0.9812 0.1766 0.0543>}
 mstr_HoodUp_back_l {parent=neck_01, loc=<5.8689 -13.1968 -6.1725>, rot=<0.9812 -0.0553 0.0543 -0.1766>}
 clavicle_l {parent=spine_05, loc=<2.5856 -0.2866 -4.3777>, rot=<0.0465 -0.7289 -0.009 -0.6829>}
 clavicle_r {parent=spine_05, loc=<2.5855 -0.2866 4.3777>, rot=<0.7289 0.0465 -0.6829 0.009>}
]
