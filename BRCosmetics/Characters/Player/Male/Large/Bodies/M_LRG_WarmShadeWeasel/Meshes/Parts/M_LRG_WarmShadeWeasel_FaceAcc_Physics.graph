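# Physics asset for the face accessory mesh of the large male WarmShadeWeasel player body, covering the front of its raised hood.
# Left and right dyn_HoodUp_Front bones carry spheres, each constrained to a smaller-sphere mstr counterpart.
# Both constraints lock swing and twist, so their set angular limits do not apply.
# A head capsule has no constraint; right-side spheres reuse left-side shape names.
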PhysicsAsset /BRCosmetics/Characters/Player/Male/Large/Bodies/M_LRG_WarmShadeWeasel/Meshes/Parts/M_LRG_WarmShadeWeasel_FaceAcc_Physics
bodies {
 dyn_HoodUp_Front_l [{shapeType=Sphere, radius=3.6393, shapeName=dyn_HoodUp_Front_l_sphere}]
 dyn_HoodUp_Front_r [{shapeType=Sphere, radius=3.6393, shapeName=dyn_HoodUp_Front_l_sphere}]
 head [{shapeType=Capsule, length=16.4511, radius=7.9625, shapeName=head_capsule, rotation=<rot 90 -167.7723 -167.7719>}]
 mstr_HoodUp_Front_l [{shapeType=Sphere, radius=0.9931, shapeName=mstr_HoodUp_Front_l_sphere}]
 mstr_HoodUp_Front_r [{shapeType=Sphere, radius=0.9931, shapeName=mstr_HoodUp_Front_l_sphere}]
}
constraints [
 {
  bone1Name=dyn_HoodUp_Front_l
  bone2Name=mstr_HoodUp_Front_l
  swing1LimitDegrees=45
  swing1Motion=Locked
  swing2LimitDegrees=45
  swing2Motion=Locked
  twistLimitDegrees=45
  twistMotion=Locked
 }
 {
  bone1Name=dyn_HoodUp_Front_r
  bone2Name=mstr_HoodUp_Front_r
  swing1LimitDegrees=45
  swing1Motion=Locked
  swing2LimitDegrees=45
  swing2Motion=Locked
  twistLimitDegrees=45
  twistMotion=Locked
 }
]
