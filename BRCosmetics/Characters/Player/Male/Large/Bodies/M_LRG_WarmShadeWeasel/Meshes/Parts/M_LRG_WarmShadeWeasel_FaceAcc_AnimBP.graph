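# Animation Blueprint for the WarmShadeWeasel face accessory, keeping its eyeglasses fitted to facial expressions.
# AnimGraph copies the attached parent's pose into a Control Rig, runs head-relative RigidBody physics, then applies two additive corrective offsets to the glasses bone.
# EventGraph caches the Head part on customization load, remapping its cheek and forehead bone movement (stronger side wins) into those correctives' alphas.
# Each update also refreshes a gravity override through a deprecated function.
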AnimBlueprint /BRCosmetics/Characters/Player/Male/Large/Bodies/M_LRG_WarmShadeWeasel/Meshes/Parts/M_LRG_WarmShadeWeasel_FaceAcc_AnimBP
target_skeleton=&/M_LRG_WarmShadeWeasel_FaceAcc_Skeleton
graphs [
 AnimGraph {nodes [
   n1 = Control Rig(Source=Copy Pose From Mesh() {bUseAttachedParent=true}) {
    alphaBoolBlend {blendOption=Linear}
    controlRigAssetReference {blueprintRigClass=&/M_LRG_WarmShadeWeasel_FaceAcc_CtrlRig.M_LRG_WarmShadeWeasel_FaceAcc_CtrlRig_C}
   }
   n2 = RigidBody!(ComponentPose=Local To Component(LocalPose=@n1), Alpha, OverrideWorldGravity:Vector="0,0,0", ExternalForce) -> (Pose) {
    bOverrideWorldGravity=true
    overridePhysicsAsset=&/M_LRG_WarmShadeWeasel_FaceAcc_Physics
    simulationSpace=BaseBoneSpace
    alphaBoolBlend {blendOption=Linear}
    baseBoneRef {boneName=head}
    componentAppliedLinearAccClamp=<400 400 400>
    componentLinearAccScale=<1 1 1>
    componentLinearVelScale=<1 1 1>
    simSpaceSettings {maxAngularAcceleration=10, maxAngularVelocity=10, maxLinearAcceleration=10, maxLinearVelocity=10, velocityScaleZ=0, worldAlpha=0.2}
   }
   n3 = "Transform (Modify) Bone - Bone: mstr_eyeglasses"(
    ComponentPose=@n2
    Alpha=Get EyeGlassesCorrectiveCheeckAlpha() -> (EyeGlassesCorrectiveCheeckAlpha:double)
    Translation="0.389,0.141,0.000"
    Rotation="0.000,-0.963,-0.000"
    Scale="1,1,1"
   ) {
    rotationMode=BMM_Additive
    rotationSpace=BCS_ParentBoneSpace
    translationMode=BMM_Additive
    translationSpace=BCS_ParentBoneSpace
    alphaBoolBlend {blendOption=Linear}
    boneToModify {boneName=mstr_eyeglasses}
    rotation {yaw=-0.963}
    translation=<0.389 0.141>
   }
   n4 = "Transform (Modify) Bone - Bone: mstr_eyeglasses"(ComponentPose=@n3, Alpha=Get GlassesCorBrowAlpha() -> (GlassesCorBrowAlpha:double), Translation="0,0.5,0", Rotation="0,0,0", Scale="1,1,1") {
    rotationMode=BMM_Additive
    rotationSpace=BCS_ParentBoneSpace
    translationMode=BMM_Additive
    translationSpace=BCS_ParentBoneSpace
    alphaBoolBlend {blendOption=Linear}
    boneToModify {boneName=mstr_eyeglasses}
    translation {y=0.5}
   }
   Output Pose(Result=Component To Local(ComponentPose=@n4)) {layerGroup=None}
  ]}
 EventGraph {nodes [
   Event BlueprintUpdateAnimation()
   > n1 = Gravity Override State Function DEPRECATED(CharacterPartAnimInstance=Self-Reference(), GameTimeInSeconds=GetGameTimeInSeconds())
   > Set Gravity Override(Gravity Override=@n1)
   n2 = Get HeadSkeleton() -> (HeadSkeleton:Skeletal Mesh Component Budgeted Object Reference)
   > Is Valid(InputObject=@n2)
   n3 = GetSocketTransform!(self=@n2, InSocketName=FACIAL_L_CheekOuter, TransformSpace=RTS_ParentBoneSpace) -> (
    ReturnValue_Location_X:double
    ReturnValue_Location_Y:double
    ReturnValue_Location_Z:double
    ReturnValue_Rotation_Roll:float
    ReturnValue_Rotation_Pitch:float
    ReturnValue_Rotation_Yaw:float
    ReturnValue_Scale:Vector
   )
   n4 = GetSocketTransform(self=@n2, InSocketName=FACIAL_R_CheekOuter, TransformSpace=RTS_ParentBoneSpace)
   > Set EyeGlassesCorrectiveCheeckAlpha(EyeGlassesCorrectiveCheeckAlpha:double="Max (Float)"(
     A=Remap Float(Input=@n3.ReturnValue_Location_Y, OldRange=(X=-0.5,Y=-3), NewRange=(X=0,Y=1), Clamp=true) -> (Output=~)
     B=Remap Float(Input=@n4.ReturnValue_Location_Y, OldRange=(X=-0.5,Y=-3), NewRange=(X=0,Y=1), Clamp=true) -> (Output=~)
    )) -> (Output_Get:double)
   n5 = Get HeadSkeleton()
   n6 = GetSocketTransform(self=@n5, InSocketName=FACIAL_R_ForeheadMid, TransformSpace=RTS_ParentBoneSpace)
   n7 = GetSocketTransform(self=@n5, InSocketName=FACIAL_L_ForeheadMid, TransformSpace=RTS_ParentBoneSpace)
   > Set GlassesCorBrowAlpha(GlassesCorBrowAlpha:double="Max (Float)"(
     A=Remap Float(Input=@n7.ReturnValue_Location_Y, OldRange=(X=-1,Y=-5), NewRange=(X=0,Y=1), Clamp=true) -> (Output=~)
     B=Remap Float(Input=@n6.ReturnValue_Location_Y, OldRange=(X=-1,Y=-5), NewRange=(X=0,Y=1), Clamp=true) -> (Output=~)
    )) -> (Output_Get:double)
   n8 = Event OnCharacterCustomizationLoaded()
   > Set HeadSkeleton(HeadSkeleton:Skeletal Mesh Component Budgeted Object Reference=Get Array Item with EFort Custom Part Type(Array=@n8.CustomMeshParts, PartType=Head)) -> (Output_Get:Skeletal Mesh Component Budgeted Object Reference)
  ]}
]
variables [
 EyeGlassesCorrectiveCheeckAlpha {type=number}
 GlassesCorBrowAlpha {type=number}
 Gravity Override {type=Vector}
 HeadSkeleton {type=SkeletalMeshComponentBudgeted}
]
